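# Niagara effect for the meteor in the WealthLamb_FateStork emote, layering timed impact flares, flashes, rock debris and dust.
# Flash, debris and dust emitters spawn from skeletal mesh sockets, with spawn groups varying bursts within emitters.
# GPU debris uses emitter-age curves to hang under high drag and low gravity, then release after about two seconds.
# Front-end flags, quality levels and time-of-day colors adjust scale, collision and tint.
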NiagaraSystem /BRCosmetics/Animation/Game/MainPlayer/Emotes/WealthLamb_FateStork/VFX/NS_Emote_WealthLamb_FateStork_Meteor
dependencies {
 usedDataInterfaces [
  &NiagaraDataInterfaceArrayFloat3
  &NiagaraDataInterfaceMeshRendererInfo
  &NiagaraDataInterfaceCurve
  &NiagaraDataInterfaceVectorCurve
  &NiagaraDataInterfacePlatformSet
  &NiagaraDataInterfaceArrayColor
  &NiagaraDataInterfaceSpriteRendererInfo
  &NiagaraDataInterfaceColorCurve
  &NiagaraDataInterfaceVector2DCurve
 ]
 usedDynamicInputs [
  &SimulationPosition
  &UniformRangedVector2D
  &MakeFloatFromVector2d
  &MakeVector2D
  &ReturnNormalizedExecIndex
 ]
 usedModules [
  &self:SystemSpawnScript.SystemScriptSource.SystemScriptGraph.NiagaraNodeAssignment_0.SetVariables_#1
  &self:SystemSpawnScript.SystemScriptSource.SystemScriptGraph.NiagaraNodeAssignment_1.SetVariables_#2
  &SystemState
 ]
}
emitters {
 Flare_Complex001 {
  emitterUpdateScript [
   &EmitterState(Life Cycle Mode=::NewEnumerator1, Loop Behavior=::NewEnumerator1, Loop Delay=0.275, Loop Duration=2, UseLoopDelay=true)
   &SpawnBurst_Instantaneous(Use Spawn Probability=false, UseLoopCountLimit=false) {module=SpawnBurst_Instantaneous003}
   &SpawnBurst_Instantaneous(Spawn Group=1, Use Spawn Probability=false, UseLoopCountLimit=false) {module=SpawnBurst_Instantaneous001}
   &SpawnBurst_Instantaneous(Spawn Group=2, Spawn Time=0.025, Use Spawn Probability=false, UseLoopCountLimit=false)
   &SpawnBurst_Instantaneous(Spawn Group=3, Spawn Time=0.025, Use Spawn Probability=false, UseLoopCountLimit=false) {module=SpawnBurst_Instantaneous002}
  ]
  particleSpawnScript [
   /SetVariables(Particles.VisibilityTag=&SpawnGroup(Max Spawn Group=2)) {isSetParameters=true}
   &InitializeParticle(
    Color=&MakeCustomLinearColorFromBool(
     Bool=&SetBoolByIntComparison(A=&SpawnGroup(), B=1, Comparison Type=::NewEnumerator7)
     True Color=&SelectLinearColorFromArray(
      Array Sampling Mode=::NewEnumerator1
      Direct Array Index=&SpawnGroup()
      Color Selection Array=DataInterface() {propertyValues {ColorData [<rgba 1 1 1 1>, <rgba 1 1 1 0.5>, <rgba 1 1 1 1>, <rgba 100 61.7074 35.1155 1>]}}
      Random Seed=?
     )
     False Color=<rgba 1 1 1 1>
    )
    Lifetime=&MaskFloatBySpawnGroup(
     Masked Float=0.5
     Passthrough Float=&MaskFloatBySpawnGroup(Masked Float=0.25, Passthrough Float=0.25, Spawn Group Mask=1)
    )
    Mesh Scale Mode=::NewEnumerator3
    Mesh Uniform Scale=&MaskFloatBySpawnGroup(Masked Float=2, Passthrough Float=1, Spawn Group Mask=2)
    Sprite Rotation Angle=&MaskFloatBySpawnGroup(Masked Float=0, Passthrough Float=&RandomRangeFloat(Minimum=0, Maximum=1))
    Sprite Rotation Mode=::NewEnumerator1
    Sprite Size=&Vector2DFromFloat(Value=500)
    Sprite Size Mode=::NewEnumerator4
    UsePositionOffset=true
    Write Material Random=true
   )
   &CameraOffset(Camera Offset Amount=&MaskFloatBySpawnGroup(Masked Float=50, Passthrough Float=80))
  ]
  particleUpdateScript [
   &ParticleState()
   &SpriteRotationRate(
    Delta Time:NiagaraFloat=&FloatFromCurve(FloatCurve=DataInterface() {propertyValues {Curve {keys [
         {tangentMode=RCTM_User, time=0, value=1}
         {arriveTangent=-0.3174, interpMode=RCIM_Cubic, leaveTangent=-0.3174, tangentMode=RCTM_User, time=0.3, value=1}
         {arriveTangent=0.0123, interpMode=RCIM_Cubic, leaveTangent=0.0123, tangentMode=RCTM_User, time=1, value=0}
        ]}}})
    Rotation Rate=&MaskFloatBySpawnGroup(Masked Float=5, Passthrough Float=0, Spawn Group Mask=1)
   )
   &ScaleSpriteSize(
    Scale Factor=&MaskVector2dBySpawnGroup(
     Masked Vector2D=&Vector2DFromCurve(
      Vector2Curve=DataInterface() {propertyValues {
        XCurve {keys [<at 0 1 RCIM_Cubic>, {arriveTangent=-3.1442, interpMode=RCIM_Cubic, leaveTangent=-3.1442, tangentMode=RCTM_User, time=1, value=0}]}
        YCurve {keys [
          {arriveTangent=4.8145, interpMode=RCIM_Cubic, time=0, value=1}
          {arriveTangent=-3.0695, interpMode=RCIM_Cubic, leaveTangent=-3.0695, tangentMode=RCTM_User, time=1, value=0}
         ]}
       }}
      ScaleCurve=<1 1>
     )
     Passthrough Vector2D=&Vector2DFromFloat(Value=&MaskFloatBySpawnGroup(
       Masked Float=&FloatFromCurve(FloatCurve=DataInterface() {propertyValues {Curve {keys [
            {arriveTangent=5.1291, interpMode=RCIM_Cubic, leaveTangent=5.1291, tangentMode=RCTM_User, time=0, value=0}
            {arriveTangentWeight=0.1274, interpMode=RCIM_Cubic, leaveTangentWeight=0.7778, tangentMode=RCTM_User, time=0.5, value=1}
           ]}}})
       Passthrough Float=&FloatFromCurve(FloatCurve=DataInterface() {propertyValues {Curve {keys [
            {arriveTangent=3.3451, interpMode=RCIM_Cubic, leaveTangent=3.3451, tangentMode=RCTM_User, time=0, value=0.5}
            {arriveTangentWeight=0.1274, interpMode=RCIM_Cubic, leaveTangentWeight=0.7778, tangentMode=RCTM_User, time=0.5, value=1}
           ]}}})
       Spawn Group Mask=1
      ))
     Spawn Group Mask=0
    )
    Scale Sprite Size Mode=::NewEnumerator2
   )
   &Color(Color=&MakeCustomLinearColorFromBool(
     Bool=&SetBoolByIntComparison(A=&SpawnGroup(), B=1, Comparison Type=::NewEnumerator7)
     True Color=&SelectLinearColorFromArray(
      Array Sampling Mode=::NewEnumerator1
      Direct Array Index=&SpawnGroup()
      Color Selection Array=DataInterface() {propertyValues {ColorData [<rgba 1 1 1 1>, <rgba 1 1 1 0.5>, <rgba 1 1 1 1>, <rgba 100 61.7074 35.1155 1>]}}
      Random Seed=?
     )
     False Color=<rgba 1 1 1 1>
    ))
   &ScaleMeshSize(Scale Factor=&VectorFromFloat(Value=&MaskFloatBySpawnGroup(
      Masked Float=&FloatFromCurve(FloatCurve=DataInterface() {propertyValues {Curve {
          defaultValue=0
          keys [
           {arriveTangent=4.1785, interpMode=RCIM_Cubic, leaveTangent=4.1785, tangentMode=RCTM_User, time=0, value=0}
           {arriveTangent=1.529, interpMode=RCIM_Cubic, leaveTangent=1.529, tangentMode=RCTM_User, time=0.2042, value=0.6228}
           {arriveTangent=0.0224, interpMode=RCIM_Cubic, leaveTangent=0.0224, tangentMode=RCTM_User, time=1, value=1}
          ]
         }}})
      Passthrough Float=&FloatFromCurve(FloatCurve=DataInterface() {propertyValues {Curve {
          defaultValue=0
          keys [
           <at 0 0 RCIM_Cubic>
           {arriveTangent=1.6664, interpMode=RCIM_Cubic, leaveTangent=1.6664, tangentMode=RCTM_User, time=0.45, value=0.2}
           {arriveTangent=1.1179, interpMode=RCIM_Cubic, leaveTangent=1.1179, tangentMode=RCTM_User, time=0.55, value=0.75}
           {arriveTangent=0.0224, interpMode=RCIM_Cubic, leaveTangent=0.0224, tangentMode=RCTM_User, time=1, value=1}
          ]
         }}})
      Spawn Group Mask=2
     )))
   &DynamicMaterialParameters(
    Index 0 Param 1=&MaskFloatBySpawnGroup(
     Masked Float=&FloatFromCurve(FloatCurve=DataInterface() {propertyValues {Curve {keys [{arriveTangent=0.1648, interpMode=RCIM_Cubic, leaveTangent=0.1648, tangentMode=RCTM_User, time=0, value=0.9}, <at 0.5 1 RCIM_Cubic>]}}})
     Passthrough Float=&FloatFromCurve(FloatCurve=DataInterface() {propertyValues {Curve {keys [<at 0 0.9 RCIM_Cubic>,<at 1 1 RCIM_Cubic>]}}})
     Spawn Group Mask=2
    )
    Index 0 Param 3=5
    Index 0 Param 4=0.04
    Module.Index0Float0Write=true
    Module.Index0Float1Write=false
    Module.Index0Float2Write=true
    Module.Index0Float3Write=true
    Module.Index1Float0Write=true
    Module.Index1Float1Write=true
    Module.Index1Float2Write=true
    Module.Index1Float3Write=true
    Module.Index2Float0Write=true
    Module.Index2Float1Write=true
    Module.Index2Float2Write=true
    Module.Index2Float3Write=true
    Write Parameter Index 1=::NewEnumerator1
    Write Parameter Index 2=::NewEnumerator1
   )
   &ScaleColor(
    Scale Alpha=&MaskFloatBySpawnGroup(
     Masked Float=&FloatFromCurve(FloatCurve=DataInterface() {propertyValues {Curve {keys [
          {arriveTangent=-23.168, interpMode=RCIM_Cubic, leaveTangent=-23.168, tangentMode=RCTM_User, time=0, value=4}
          {arriveTangent=-4.5789, interpMode=RCIM_Cubic, leaveTangent=-4.5789, tangentMode=RCTM_User, time=0.2, value=1}
          <at 1 0 RCIM_Cubic>
         ]}}})
     Passthrough Float=&MaskFloatBySpawnGroup(
      Masked Float=&FloatFromCurve(
       FloatCurve=DataInterface() {propertyValues {Curve {
          defaultValue=0
          keys [
           {arriveTangent=9.0725, interpMode=RCIM_Cubic, leaveTangent=9.0725, tangentMode=RCTM_User, time=0, value=0}
           {arriveTangent=-2.5249, interpMode=RCIM_Cubic, leaveTangent=-2.5249, tangentMode=RCTM_User, time=0.1408, value=0.9847}
           {arriveTangent=-0.0405, interpMode=RCIM_Cubic, leaveTangent=-0.0405, tangentMode=RCTM_User, time=1, value=0}
          ]
         }}}
       Scale Curve=0.2
      )
      Passthrough Float=&MaskFloatBySpawnGroup(
       Masked Float=&FloatFromCurve(
        FloatCurve=DataInterface() {propertyValues {Curve {keys [<at 0 1 RCIM_Cubic>,<at 1 0 RCIM_Cubic>]}}}
        Scale Curve=&MaskFloatBySpawnGroup(Masked Float=0.2, Passthrough Float=1, Spawn Group Mask=2)
       )
       Passthrough Float=&FloatFromCurve(
        FloatCurve=DataInterface() {propertyValues {Curve {keys [<at 0 1 RCIM_Cubic>,<at 1 0 RCIM_Cubic>]}}}
        Scale Curve=&MaskFloatBySpawnGroup(Masked Float=0.2, Passthrough Float=1, Spawn Group Mask=2)
       )
       Spawn Group Mask=2
      )
      Spawn Group Mask=1
     )
    )
    Scale RGB=&MakeVectorFromLinearColorRGB(Color=&MakeCustomLinearColorFromBool(
      Bool=&SetBoolByIntComparison(A=&SpawnGroup(), B=1, Comparison Type=::NewEnumerator7)
      True Color=&ColorFromCurve(ColorCurve=DataInterface() {propertyValues {
         AlphaCurve {keys []}
         BlueCurve {keys [{arriveTangent=-1.5872, interpMode=RCIM_Cubic, leaveTangent=-1.5872, tangentMode=RCTM_User, time=0, value=3}, <at 1 0.5 RCIM_Cubic>]}
         GreenCurve {keys [{arriveTangent=-4.9882, interpMode=RCIM_Cubic, leaveTangent=-4.9882, tangentMode=RCTM_User, time=0, value=3.75}, <at 1 0.625 RCIM_Cubic>]}
         RedCurve {keys [{arriveTangent=-3.5785, interpMode=RCIM_Cubic, leaveTangent=-3.5785, tangentMode=RCTM_User, time=0, value=4}, <at 1 1 RCIM_Cubic>]}
        }})
      False Color=<rgba 1 1 1 1>
     ))
    ScaleA=true
    ScaleRGB=true
   )
  ]
  rendererClasses [
   &NiagaraSpriteRendererProperties
   &NiagaraLightRendererProperties
   &NiagaraMeshRendererProperties
  ]
  renderers [
   {rendererClass=&NiagaraSpriteRendererProperties}
   {rendererClass=&NiagaraLightRendererProperties, rendererIndex=1}
   {rendererClass=&NiagaraMeshRendererProperties, rendererIndex=2}
  ]
 }
 FlashElements001 {
  emitterUpdateScript [
   &EmitterState(Life Cycle Mode=::NewEnumerator1, Loop Behavior=::NewEnumerator1, Loop Delay=0.2, Loop Duration=2, UseLoopDelay=true)
   &SpawnBurst_Instantaneous(Spawn Count=3, Spawn Group=1, Spawn Time=0.02, Use Spawn Probability=false, UseLoopCountLimit=false) {module=SpawnBurst_Instantaneous001}
   &SpawnBurst_Instantaneous(
    Spawn Count=&RandomRangeInt(Minimum=6, Maximum=8)
    Use Spawn Probability=false
    UseLoopCountLimit=false
   )
  ]
  particleSpawnScript [
   &InitializeParticle(Color Mode=::NewEnumerator0, Lifetime=2, Sprite Rotation Mode=::NewEnumerator2, UsePositionOffset=false, Write Material Random=true)
   &SkeletalMeshLocation(
    Bone / Socket Index=3
    MaskSpawnGroup=false
    Mesh Sampling Type=::NewEnumerator5
    Orientation Sampling=::NewEnumerator1
    Position Sampling=::NewEnumerator2
    Sampled Position Offset=<10 -5 -5>
    Sampled Position Offset Coordinate Space:ENiagaraSkelMeshTransforms=::NewEnumerator2
    Skeletal Mesh=$System.Skeletal Mesh
    Socket Sampling Mode=::NewEnumerator5
    Use Sampled Mesh Attribute as Ribbon ID=false
   ) {module=SkeletalMeshLocation001}
   &ShapeLocation(
    MaskSpawnGroup=false
    Override Seed=false
    Shape Origin=$Particles.SkeletalMeshLocation001.SampledPosition
    Sphere Radius=5
    UseOwnerScale=true
   )
   &SpriteFacingAndAlignment(Sprite Alignment=$Particles.ShapeLocation.ShapeNormal, WriteAlignment=true, WriteFacing=false)
   &FN_PerParticleLOD(
    Enable SpriteSize Distance Scaling=true
    Sprite Size Scale Far=&RandomRangeVector2D(Minimum=<1.5 1.5>, Maximum=<3 3>)
    Sprite Size Scale Near:Vector2f=&RandomRangeVector2D(Minimum=<1.2 1.2>, Maximum=<2 2>)
   )
   /SetVariables(Particles.SpriteSize=<1 1>, Particles.VisibilityTag=&SpawnGroup()) {isSetParameters=true}
   /SetVariables(
    Particles.Flash Star Cam Offset=15
    Particles.Flash Star Color:LinearColor=$System.TOD Color
    Particles.Flash Star Lifetime=1
    Particles.Flash Star Rotation:NiagaraFloat=&RandomRangeFloat(Minimum=0, Maximum=360)
    Particles.Flash Star Sprite Size:Vector2f=&Multiply_Vector2D_ByFloat(
     Float=&Multiply_Float(
      A=&MakeCustomFloatFromBool(Bool=$User.bIsFrontEnd, True Float=$User.FrontEndScale, False Float=1)
      B=&RandomRangeFloat(Minimum=110, Maximum=130)
     )
     Vector2D=$Particles.SpriteSize
    )
    Particles.Flash Star SubUV Index:NiagaraFloat=&RandomRangeFloat(Minimum=0, Maximum=3)
   ) {isSetParameters=true}
   /SetVariables(
    Particles.Flash Flare Color:LinearColor=$System.TOD Color
    Particles.Flash Flare Lifetime=0.45
    Particles.Flash Flare Sprite Size:Vector2f=&Multiply_Vector2D_ByFloat(Float=&FN_FrontEndFloatModifier(Gameplay=1, Front End=$User.FrontEndScale), Vector2D=<120 120>)
   ) {isSetParameters=true}
   &DynamicMaterialParameters(
    Index 0 Param 1=0.2
    Index 0 Param 2=1
    Module.Index0Param0Write=true
    Module.Index0Param1Write=true
    Module.Index0Param2Write=true
    Module.Index0Param3Write=true
   )
   &CameraOffset(Camera Offset Amount=15)
   &SpriteFacingAndAlignment(Sprite Facing=&RandomVector(Vector Scale=1), WriteAlignment=false, WriteFacing=true) {module=SpriteFacingAndAlignment001}
  ]
  particleUpdateScript [
   &ParticleState()
   /SetVariables(
    Particles.Flash Flare Age:NiagaraFloat=&Divide_Float(A=$Particles.Age, B=$Particles.Initial.Flash Flare Lifetime)
    Particles.Flash Star Age:NiagaraFloat=&Divide_Float(A=$Particles.Age, B=$Particles.Initial.Flash Star Lifetime)
   ) {isSetParameters=true}
   /SetVariables(
    Particles.Flash Flare Color:LinearColor=&MakeLinearColorFromVectorAndFloat(
     "Vector (RGB)"=&Multiply_Vector(
      A=&Multiply_Vector(
       A=&MakeVectorFromLinearColorRGB(Color=$Particles.Initial.Flash Flare Color)
       B=&MakeVectorFromLinearColorRGB(Color=&ColorFromCurve(
         ColorCurve=DataInterface() {propertyValues {
           AlphaCurve {keys [<at 0 1 RCIM_Cubic>,<at 1 0 RCIM_Cubic>]}
           BlueCurve {keys [<at 0.0198 0.3>,<at 1 0.4>]}
           GreenCurve {keys [<at 0.0198 1.4333>,<at 1 1.4667>]}
           RedCurve {keys [<at 0.0198 2>,<at 1 2>]}
          }}
         CurveIndex=$Particles.Flash Flare Age
        ))
      )
      B=&VectorFromFloat(Value=&FN_FrontEndFloatModifier(Gameplay=1, Front End=0.75))
     )
     "Float (Alpha)"=&Multiply_Float(
      A=&MakeFloatFromLinearColor(LinearColor=<rgba 1 1 1 1>, Channel:ENiagaraLinearColor_Channels=::NewEnumerator3)
      B=&FloatFromCurve(
       FloatCurve=DataInterface() {propertyValues {Curve {keys [{arriveTangent=-1.9727, interpMode=RCIM_Cubic, time=0.1, value=1}, <at 1 0 RCIM_Cubic>]}}}
       CurveIndex=$Particles.Flash Flare Age
      )
     )
    )
    Particles.Flash Star Color:LinearColor=&MultiplyLinearColor(
     Linear Color=$Particles.Initial.Flash Star Color
     Scale Factor=&MakeVector4_FromLinearColor(LinearColor=&ColorFromCurve(
       ColorCurve=DataInterface() {propertyValues {
         AlphaCurve {keys [<at 0 1 RCIM_Cubic>,<at 1 0 RCIM_Cubic>]}
         BlueCurve {keys [<at 0.0198 0.3>,<at 1 0.4>]}
         GreenCurve {keys [<at 0.0198 1.5417>,<at 1 1.6>]}
         RedCurve {keys [<at 0.0198 2>,<at 1 2>]}
        }}
       CurveIndex=$Particles.Flash Flare Age
      ))
    )
    Particles.Flash Star Sprite Size:Vector2f=&Multiply_Vector2D_ByFloat(
     Float=&FloatFromCurve(
      FloatCurve=DataInterface() {propertyValues {Curve {keys [
          {arriveTangent=2.7013, interpMode=RCIM_Cubic, leaveTangent=2.7013, tangentMode=RCTM_User, time=0, value=0}
          {interpMode=RCIM_Cubic, tangentMode=RCTM_User, time=1, value=1}
         ]}}}
      CurveIndex=$Particles.Flash Star Age
     )
     Vector2D=$Particles.Initial.Flash Star Sprite Size
    )
   ) {isSetParameters=true}
   &FN_Light_Attributes(
    Color=&MultiplyLinearColor(
     Linear Color=<rgba 1 0.7875 0.15 1>
     Scale Factor=&MakeVector4_FromLinearColor(LinearColor=&LerpLinearColors(StartColor=<rgba 1 1 1 1>, EndColor=$System.TOD Color, LerpFactor=2))
    )
    Create Color Attribute=true
    Create Exponent Attribute=true
    Create Position Attribute=true
    Create Radius Attribute=true
    Exponent=1
    Radius=&FN_FrontEndFloatModifier(Gameplay=90, Front End=90)
    Scale Color=&FloatFromCurve(
     FloatCurve=DataInterface() {propertyValues {Curve {keys [
         {arriveTangent=-0.5504, interpMode=RCIM_Cubic, leaveTangent=-0.5504, tangentMode=RCTM_User, time=0, value=1}
         {arriveTangent=-0.5426, interpMode=RCIM_Cubic, leaveTangent=-0.5426, tangentMode=RCTM_User, time=0.55, value=0}
        ]}}}
     Scale Curve=2
    )
   )
  ]
  rendererClasses [&NiagaraSpriteRendererProperties]
  renderers [
   {rendererClass=&NiagaraSpriteRendererProperties}
   {rendererClass=&NiagaraSpriteRendererProperties, rendererIndex=1}
  ]
 }
 HT_Debris {
  simTarget=GPUComputeSim
  emitterUpdateScript [
   &EmitterState()
   &SpawnRate(
    SpawnGroup=2
    SpawnRate=&FloatFromCurve(
     FloatCurve=DataInterface() {propertyValues {Curve {keys [<at 0 0 RCIM_Constant>,<at 0.38 1>,<at 1.3 0>]}}}
     CurveIndex=$Emitter.LoopedAge
     Scale Curve=30
    )
    Use Spawn Probability=false
   )
   &SpawnRate(
    SpawnGroup=1
    SpawnRate=&FloatFromCurve(
     FloatCurve=DataInterface() {propertyValues {Curve {keys [<at 0 0 RCIM_Constant>,<at 0.38 1>,<at 1.3 0>]}}}
     CurveIndex=$Emitter.LoopedAge
     Scale Curve=30
    )
    Use Spawn Probability=false
   ) {module=SpawnRate001}
   &SpawnBurst_Instantaneous(Spawn Count=40, Spawn Group=2, Spawn Time=0.38, Use Spawn Probability=false, UseLoopCountLimit=false)
   &SpawnBurst_Instantaneous(
    Spawn Count=&RandomRangeInt(Minimum=50, Maximum=60)
    Spawn Group=1
    Spawn Time=0.38
    Use Spawn Probability=false
    UseLoopCountLimit=false
   ) {module=SpawnBurst_Instantaneous002}
  ]
  particleSpawnScript [
   &InitializeParticle(
    Color=&SelectLinearColorFromArray(
     Array Sampling Mode=::NewEnumerator0
     Direct Array Index=?
     Color Selection Array=DataInterface() {propertyValues {ColorData [
        <rgba 0.1356 0.1141 0.0991 1>
        <rgba 0.2623 0.2122 0.1095 1>
        <rgba 0.147 0.1105 0.0895 1>
        <rgba 0.1651 0.1441 0.1022 1>
        <rgba 0.0939 0.082 0.0581 1>
        <rgba 0.1148 0.0929 0.0479 1>
       ]}}
     Random Seed=0
    )
    Lifetime Max=5
    Lifetime Min=4
    Lifetime Mode=::NewEnumerator1
    Mesh Renderer Array Visibility Mode=::NewEnumerator3
    Mesh Renderer Info=DataInterface() {propertyValues {MeshRenderer=&self:HT_Debris_4.NiagaraMeshRendererProperties_0}}
    Mesh Scale Max=<0.25 0.25 0.25>
    Mesh Scale Min=<0.1 0.1 0.1>
    Mesh Scale Mode=::NewEnumerator2
    Sprite Rotation Mode=::NewEnumerator2
    Sprite Size Max=<2 1>
    Sprite Size Min=<1.25 0.75>
    Sprite Size Mode=::NewEnumerator2
    UsePositionOffset=false
    Write Material Random=true
   ) {module=InitializeParticle001}
   &SkeletalMeshLocation(
    MaskSpawnGroup=false
    Mesh Sampling Type=::NewEnumerator5
    Orientation Sampling=::NewEnumerator1
    Position Sampling=::NewEnumerator2
    Sampled Position Offset=<-5 15 13>
    Sampled Position Offset Coordinate Space:ENiagaraSkelMeshTransforms=::NewEnumerator2
    Skeletal Mesh=$System.Skeletal Mesh
    Socket Sampling Mode=::NewEnumerator5
   ) {module=SkeletalMeshLocation002}
   &SkeletalMeshLocation(
    Bone / Socket Index=1
    MaskSpawnGroup=false
    Mesh Sampling Type=::NewEnumerator5
    Orientation Sampling=::NewEnumerator1
    Position Sampling=::NewEnumerator2
    Sampled Position Offset Coordinate Space:ENiagaraSkelMeshTransforms=::NewEnumerator2
    Skeletal Mesh=$System.Skeletal Mesh
    Socket Sampling Mode=::NewEnumerator5
   ) {module=SkeletalMeshLocation003}
   &SkeletalMeshLocation(
    Bone Sampling Mode=::NewEnumerator0
    MaskSpawnGroup=false
    Orientation Sampling=::NewEnumerator1
    OverrideRandomSeed=false
    Sampled Position Offset Coordinate Space:ENiagaraSkelMeshTransforms=::NewEnumerator2
    Skeletal Mesh=$System.Skeletal Mesh
   ) {enabled=false}
   /SetVariables(Particles.RockSize:NiagaraFloat=&RandomRangeFloat(Minimum=0.7, Maximum=2.5)) {isSetParameters=true}
   /SetVariables(Particles.SpriteSize=&Multiply_Vector2D_ByFloat(Float=$Particles.RockSize, Vector2D=$Particles.SpriteSize)) {isSetParameters=true}
   /SetVariables(Particles.Scale=&Multiply_VectorByFloat(Vector=$Particles.Scale, Float=$Particles.RockSize)) {isSetParameters=true}
   &ShapeLocation(
    Hemisphere Distribution=<0.5 1>
    MaskSpawnGroup=false
    Offset=<10 -25 3>
    Offset Coordinate Space=::Local
    Offset Mode=::NewEnumerator0
    Override Seed=false
    Shape Origin=$Particles.SkeletalMeshLocation002.SampledPosition
    Sphere Radius=&FloatFromCurve(FloatCurve=DataInterface() {propertyValues {Curve {keys [<at 0.3 0.7>,<at 1 1.5>]}}}, CurveIndex=$Emitter.LoopedAge, Scale Curve=22.5)
    UseOwnerScale=true
   )
   &AddVelocity(
    Cone Angle=35
    Cone Axis=<0 -1 0>
    Module.UseConeFalloff=false
    Module.UseVelDistribution=false
    Override Seed=false
    Velocity Mode=::NewEnumerator2
    Velocity Speed=&RandomRangeFloat(Minimum=50, Maximum=100)
   )
   &ApplyInitialForces()
   /Niagara/Modules/Update/SubUV/SubUVAnimation(Random SubUV Lookup Index:NiagaraFloat=$Particles.RandomFloat, SubUV Animation Mode=::NewEnumerator1) {module=SubUVAnimation001}
   /SetVariables(Particles.Velocity=&Multiply_VectorByFloat(
     Vector=$Particles.Velocity
     Float=&FloatFromCurve(FloatCurve=DataInterface() {propertyValues {Curve {keys [<at 0.4 2>,<at 1.4 1>]}}}, CurveIndex=$Particles.RockSize)
    )) {isSetParameters=true}
   /SetVariables(Particles.VisibilityTag=&SpawnGroup()) {isSetParameters=true}
  ]
  particleUpdateScript [
   /SetVariables(Particles.bIsBreaking:NiagaraBool=&SetBoolByFloatComparison(A=$Emitter.Age, B=0.35)) {isSetParameters=true}
   &ParticleState() {module=UpdateAge}
   &PointForce(
    Force Origin:NiagaraPosition=$StackContext.SkeletalMeshLocation002.SampledPosition
    Force Strength=&FloatFromCurve(
     FloatCurve=DataInterface() {propertyValues {Curve {keys [
         <at 0 0.3>
         {arriveTangent=1.1197, interpMode=RCIM_Cubic, leaveTangent=1.1197, tangentMode=RCTM_User, time=1.84, value=0.5}
         {arriveTangent=0.5466, tangentMode=RCTM_User, time=2.5, value=1.3}
        ]}}}
     CurveIndex=$Emitter.Age
     Scale Curve=&RandomRangeFloat(Minimum=200, Maximum=800)
    )
    Module.Use Falloff=false
    Module.UseLinearForce=false
   )
   &Drag(
    Drag=&FloatFromCurve(
     FloatCurve=DataInterface() {propertyValues {Curve {keys [<at 0 11 RCIM_Constant>,<at 2.4 0.2 RCIM_Constant>]}}}
     CurveIndex=$Emitter.Age
     Scale Curve=&RandomRangeFloat(Minimum=1, Maximum=1.2)
    )
    Use Linear Drag=true
    Use Rotational Drag=true
   )
   &GravityForce(Gravity=&Multiply_VectorByFloat(
     Vector=<0 0 -980>
     Float=&FloatFromCurve(
      FloatCurve=DataInterface() {propertyValues {Curve {keys [<at 0.1 0.075 RCIM_Constant>,<at 2.65 0.5 RCIM_Constant>]}}}
      CurveIndex=$Emitter.Age
     )
    ))
   &Collision(
    Advanced Aging Rate=1
    Age Colliding Particles=false
    Collision Enabled=&Boolean_OrOperation(
     Bool1=&IsPlatformSetActive(New Platform Set=DataInterface() {propertyValues {Platforms {qualityLevelMask=-4}}})
     Bool2=$User.bIsFrontEndPreview
    )
    Control Roll On Collision=true
    Enable Rest State=false
    Friction=0.7
    GPU Collision Type=::NewEnumerator2
    Module.EnableMaxCollisionCount=false
    Module.Randomize Collision Normal Vector=true
    Particle Radius Scale=0.5
    Restitution=0.5
    ScaleRateOfRotation=false
   )
   &AddRotationalVelocity(Rotation Rate=&Multiply_VectorByFloat(
     Vector=&RandomVector(Vector Scale=&RandomRangeFloat(Minimum=0.5, Maximum=1))
     Float=&FloatFromCurve(
      FloatCurve=DataInterface() {propertyValues {Curve {keys [<at 0 0.2 RCIM_Constant>,<at 2.3 1 RCIM_Constant>]}}}
      CurveIndex=$Emitter.Age
      Scale Curve=&RandomRangeFloat(Minimum=1, Maximum=1.2)
     )
    ))
   &SolveForcesAndVelocity(Clamp Velocity=false, Limit Acceleration=false)
   &ScaleSpriteSize(
    Uniform Curve Scale=2
    Uniform Curve Sprite Scale=DataInterface() {propertyValues {Curve {keys [
        <at 0 0>
        <at 0.05 1>
        {arriveTangent=-1.0526, interpMode=RCIM_Cubic, leaveTangent=-1.0526, time=0.8, value=1}
        {arriveTangent=-5.7307, interpMode=RCIM_Cubic, leaveTangent=-5.7307, tangentMode=RCTM_User, time=1, value=0}
       ]}}}
   )
   &ScaleSpriteSize(
    Uniform Curve Index=&NormalizeDistanceRange(
     Start Position:NiagaraPosition=$Engine.Emitter.SimulationPosition
     End Position:NiagaraPosition=$Particles.Position
     Distance:NiagaraFloat=700
     Invert Normalized Range:NiagaraBool=false
    )
    Uniform Curve Scale=2
    Uniform Curve Sprite Scale=DataInterface() {propertyValues {Curve {keys [
        <at 0 1 RCIM_Cubic>
        {arriveTangent=-1, interpMode=RCIM_Cubic, leaveTangent=-1, time=0.2, value=0.9}
        {interpMode=RCIM_Cubic, leaveTangent=-5.7307, time=1, value=0}
       ]}}}
   ) {module=ScaleSpriteSize001}
   &ScaleMeshSize(Scale Factor=&VectorFromFloat(Value=&FloatFromCurve(FloatCurve=DataInterface() {propertyValues {Curve {keys [
          <at 0 0>
          <at 0.05 1>
          {arriveTangent=-1.0526, interpMode=RCIM_Cubic, leaveTangent=-1.0526, time=0.8, value=1}
          {arriveTangent=-5.7307, interpMode=RCIM_Cubic, leaveTangent=-5.7307, tangentMode=RCTM_User, time=1, value=0}
         ]}}})))
   &ScaleMeshSize(Scale Factor=&VectorFromCurve(
     VectorCurve=DataInterface() {propertyValues {
       XCurve {keys [<at 0 1 RCIM_Cubic>, {arriveTangent=-1, interpMode=RCIM_Cubic, leaveTangent=-1, time=0.3, value=0.9}, <at 1 0 RCIM_Cubic>]}
       YCurve {keys [<at 0 1 RCIM_Cubic>, {arriveTangent=-1, interpMode=RCIM_Cubic, leaveTangent=-1, time=0.3, value=0.9}, <at 1 0 RCIM_Cubic>]}
       ZCurve {keys [<at 0 1 RCIM_Cubic>, {arriveTangent=-1, interpMode=RCIM_Cubic, leaveTangent=-1, time=0.3, value=0.9}, <at 1 0 RCIM_Cubic>]}
      }}
     CurveIndex=&NormalizeDistanceRange(
      Start Position:NiagaraPosition=$Engine.Emitter.SimulationPosition
      End Position:NiagaraPosition=$Particles.Position
      Distance:NiagaraFloat=700
      Invert Normalized Range:NiagaraBool=false
     )
    )) {module=ScaleMeshSize001}
   &SpriteRotationRate(Rotation Rate=&Multiply_Float(
     A=&RandomRangeFloat(Minimum=-300, Maximum=300)
     B=&FloatFromCurve(
      FloatCurve=DataInterface() {propertyValues {Curve {keys [<at 0 0.2 RCIM_Constant>,<at 2.3 1 RCIM_Constant>]}}}
      CurveIndex=$Emitter.Age
      Scale Curve=&RandomRangeFloat(Minimum=1, Maximum=1.2)
     )
    )) {module=SpriteRotationRate001}
   &KillParticlesInVolume(Sphere Radius=30, Volume Origin:NiagaraPosition=$Particles.SkeletalMeshLocation003.SampledPosition)
  ]
  rendererClasses [&NiagaraMeshRendererProperties, &NiagaraSpriteRendererProperties]
  renderers [
   {rendererClass=&NiagaraMeshRendererProperties, rendererIndex=0}
   {rendererClass=&NiagaraSpriteRendererProperties, rendererIndex=1}
  ]
 }
 ImpactDust {
  emitterUpdateScript [
   &EmitterState()
   &SpawnBurst_Instantaneous(Spawn Count=6, Spawn Time=0.25, Use Spawn Probability=false, UseLoopCountLimit=false)
   /SetVariables(Emitter.SizeRad=25) {isSetParameters=true}
  ]
  particleSpawnScript [
   &SkeletalMeshLocation(
    Bone / Socket Index=3
    MaskSpawnGroup=false
    Mesh Sampling Type=::NewEnumerator5
    Orientation Sampling=::NewEnumerator1
    Position Sampling=::NewEnumerator2
    Sampled Position Offset Coordinate Space:ENiagaraSkelMeshTransforms=::NewEnumerator2
    Skeletal Mesh=$System.Skeletal Mesh
    Socket Sampling Mode=::NewEnumerator5
    Use Sampled Mesh Attribute as Ribbon ID=false
   ) {module=SkeletalMeshLocation004}
   /SetVariables(Particles.DustColor:LinearColor=&MaskLinearColor(
     Mask Source=0
     Mask Parameter=?
     Mask=2
     Masked Linear Color=<rgba 0.5261 0.5261 0.5261 1>
     Passthrough Linear Color=<rgba 0.4062 0.4062 0.4062 1>
    )) {isSetParameters=true}
   /SetVariables(
    Particles.SizeMax:NiagaraFloat=&RandomRangeFloat(Minimum=262.5, Maximum=375)
    Particles.SizeMin:NiagaraFloat=&RandomRangeFloat(Minimum=225, Maximum=281.25)
   ) {isSetParameters=true}
   /SetVariables(Particles.SizeCurve:NiagaraFloat=&Lerp_Float(
     A=$Particles.SizeMin
     B=$Particles.SizeMax
     Alpha=&FloatFromCurve(
      FloatCurve=DataInterface() {propertyValues {Curve {keys [
          {interpMode=RCIM_Cubic, tangentMode=RCTM_Break, time=0, value=0}
          {arriveTangentWeight=0.9217, interpMode=RCIM_Cubic, tangentMode=RCTM_Break, time=1, value=1}
         ]}}}
      CurveIndex=&RandomRangeFloat(Minimum=0, Maximum=1)
     )
    )) {isSetParameters=true}
   /SetVariables(Particles.InitSpriteRotRate=&UniformRangedFloat(Minimum=-1, Maximum=1)) {isSetParameters=true}
   &InitializeParticle(
    Color=$Particles.DustColor
    Lifetime=&MaskFloatBySpawnGroup(
     Masked Float=&RandomRangeFloat(Minimum=4, Maximum=5)
     Passthrough Float=&MaskFloatBySpawnGroup(Masked Float=&RandomRangeFloat(Minimum=2, Maximum=2.5), Passthrough Float=1, Spawn Group Mask=2)
    )
    Sprite Rotation Mode=::NewEnumerator2
    Sprite Size Mode=::NewEnumerator3
    Uniform Sprite Size=$Particles.SizeCurve
    UsePositionOffset=false
    Write Material Random=true
   )
   &ShapeLocation(
    MaskSpawnGroup=false
    Offset=&MakeCustomVectorFromBool(
     Bool=&SetBoolByIntComparison(A=&ReturnExecIndex(), B=3, Comparison Type=::NewEnumerator7)
     True Vector=<10 20 -5>
     False Vector=<40 10 0>
    )
    Offset Coordinate Space=::Local
    Offset Mode=::NewEnumerator0
    Override Seed=false
    Shape Origin=$Particles.SkeletalMeshLocation004.SampledPosition
    Sphere Radius=25
    UseOwnerScale=true
   )
   &AddVelocity(
    Cone Angle=80
    Cone Axis=<0 -1 0.1>
    Module.UseConeFalloff=false
    Module.UseVelDistribution=false
    Override Seed=false
    Velocity Mode=::NewEnumerator2
    Velocity Speed=&RandomRangeFloat(Minimum=150, Maximum=400)
   )
   /SetVariables(Particles.Velocity=&Multiply_Vector(A=$Particles.Velocity, B=<1 1 0.5>)) {isSetParameters=true}
   &CameraOffset(Camera Offset Amount=&RandomRangeFloat(Minimum=15, Maximum=25))
   &ApplyInitialForces(Apply Force to Position=true)
   &AddVelocity(
    Velocity=<0 50 0>
    Velocity Speed Scale=&MakeCustomFloatFromBool(Bool=$User.bIsFrontEndPreview, True Float=1, False Float=0)
   ) {module=AddVelocity001}
   /Niagara/Modules/Update/SubUV/V2/SubUVAnimation(
    Evaluation Type=::NewEnumerator0
    "Sprite Renderer "=DataInterface() {propertyValues {SpriteRenderer=&self:ImpactDust_0.NiagaraSpriteRendererProperties_1}}
    SubUV Animation Mode:ENiagara_SubUVLookupModeV2=::NewEnumerator2
    UseEndFrame=false
    UseStartFrame=false
   )
   &SubUVAnimation(
    Evaluation Type=::NewEnumerator0
    "Sprite Renderer "=DataInterface() {propertyValues {SpriteRenderer=&self:ImpactDust_0.NiagaraSpriteRendererProperties_2}}
    SubUV Animation Mode:ENiagara_SubUVLookupModeV2=::NewEnumerator2
    UseEndFrame=false
    UseStartFrame=false
   ) {module=SubUVAnimation001}
  ]
  particleUpdateScript [
   &ParticleState()
   &Drag(
    Drag=&FloatFromCurve(FloatCurve=DataInterface() {propertyValues {Curve {keys [<at 0 8 RCIM_Constant>,<at 2.25 0.75 RCIM_Constant>]}}}, CurveIndex=$Emitter.Age)
    Use Linear Drag=true
    Use Rotational Drag=false
   )
   &PointForce(
    Force Origin:NiagaraPosition=$Particles.Initial.ShapeLocation.ShapePosition
    Force Strength=&FloatFromCurve(
     FloatCurve=DataInterface() {propertyValues {Curve {keys [<at 0.1 0.7 RCIM_Constant>,<at 2.3 0.5 RCIM_Constant>]}}}
     CurveIndex=$Emitter.Age
     Scale Curve=&RandomRangeFloat(Minimum=300, Maximum=450)
    )
    Module.Use Falloff=false
    Module.UseLinearForce=false
   )
   &SpriteRotationRate(Rotation Rate=&Multiply_Float(
     A=$Particles.InitSpriteRotRate
     B=&FloatFromCurve(
      FloatCurve=DataInterface() {propertyValues {Curve {keys [
          {arriveTangent=-3.9021, interpMode=RCIM_Cubic, leaveTangent=-3.9021, leaveTangentWeight=1.3778, tangentMode=RCTM_User, time=0, value=1}
          {arriveTangentWeight=0.3436, interpMode=RCIM_Cubic, tangentMode=RCTM_User, time=0.75, value=0.1}
         ]}}}
      Scale Curve=50
     )
    ))
   &CurlNoiseForce(
    Bias Noise=false
    Noise Frequency=5
    Noise Strength=&FloatFromCurve(
     FloatCurve=DataInterface() {propertyValues {Curve {
        defaultValue=1
        keys [<at 0 0 RCIM_Cubic>, {arriveTangent=1.5, interpMode=RCIM_Cubic, leaveTangent=1.5, time=0.7, value=0.2}, <at 1 1.5 RCIM_Cubic>]
       }}}
     Scale Curve=75
    )
    Pan Noise=false
   ) {module=CurlNoiseForce001}
   &GravityForce(Gravity=&Multiply_VectorByFloat(
     Vector=<0 0 1>
     Float=&FloatFromCurve(
      FloatCurve=DataInterface() {propertyValues {Curve {
         defaultValue=1
         keys [<at 0 0 RCIM_Cubic>, {arriveTangent=1.5, interpMode=RCIM_Cubic, leaveTangent=1.5, time=0.5, value=0.6}, <at 1 1.5 RCIM_Cubic>]
        }}}
      Scale Curve=25
     )
    )) {module=GravityForce001}
   &SolveForcesAndVelocity(Clamp Velocity=false, Limit Acceleration=false)
   &ScaleSpriteSize(
    Scale Sprite Size Mode=::NewEnumerator0
    Uniform Scale Factor=&FloatFromCurve(
     FloatCurve=DataInterface() {propertyValues {Curve {keys [
         <at 0 0.25 RCIM_Cubic>
         {arriveTangent=0.3289, interpMode=RCIM_Cubic, leaveTangent=0.8316, tangentMode=RCTM_Break, time=2.2, value=0.75}
         <at 5 2.5 RCIM_Cubic>
        ]}}}
     CurveIndex=$Emitter.Age
    )
   ) {module=ScaleSpriteSize001}
   &DynamicMaterialParameters(
    Index 0 Param 1=&Add_Float(
     A=&RandomRangeFloat(Minimum=0, Maximum=1)
     B=&FloatFromCurve(
      FloatCurve=DataInterface() {propertyValues {Curve {keys [<at 0 0 RCIM_Cubic>, {arriveTangent=0.1107, leaveTangent=0.1107, time=2.5, value=0.25}, <at 5 1.05 RCIM_Cubic>]}}}
      CurveIndex=&Add_Float(A=&Multiply_Float(A=$Particles.Initial.MaterialRandom, B=1), B=$Emitter.Age)
      Scale Curve=0.1
     )
    )
    Index 0 Param 2=&FloatFromCurve(FloatCurve=DataInterface() {propertyValues {Curve {keys [<at 0 1>,<at 1 0>]}}})
    Index 0 Param 3=&FloatFromCurve(
     FloatCurve=DataInterface() {propertyValues {Curve {keys [
         {interpMode=RCIM_Cubic, leaveTangent=16.5652, leaveTangentWeight=0.4385, tangentMode=RCTM_Break, time=0, value=0}
         {arriveTangentWeight=0.0333, interpMode=RCIM_Cubic, leaveTangentWeight=0.9477, tangentMode=RCTM_Break, time=0.1, value=1}
         {arriveTangent=-1.7159, interpMode=RCIM_Cubic, leaveTangent=-1.7159, tangentMode=RCTM_User, time=0.4, value=0.8}
         {arriveTangent=0.2347, interpMode=RCIM_Cubic, leaveTangent=0.2723, tangentMode=RCTM_Break, time=0.6, value=0}
        ]}}}
     Scale Curve=0
    )
    Index 0 Param 4=&FloatFromCurve(FloatCurve=DataInterface() {propertyValues {Curve {keys [{interpMode=RCIM_Cubic, tangentMode=RCTM_Break, time=0.5, value=0.2}, {interpMode=RCIM_Cubic, tangentMode=RCTM_Break, time=1, value=0}]}}})
    Index 1 Param 1=&FloatFromCurve(FloatCurve=DataInterface() {propertyValues {Curve {keys [
         {interpMode=RCIM_Cubic, leaveTangentWeight=0.0833, tangentMode=RCTM_Break, time=0, value=0}
         {arriveTangentWeight=0.0873, interpMode=RCIM_Cubic, tangentMode=RCTM_Break, time=0.25, value=1}
        ]}}})
    Index 1 Param 2=&FloatFromCurve(
     FloatCurve=DataInterface() {propertyValues {Curve {keys [
         {interpMode=RCIM_Cubic, tangentMode=RCTM_Break, time=0, value=0.2}
         {interpMode=RCIM_Cubic, leaveTangent=0.9066, tangentMode=RCTM_Break, time=1, value=2}
        ]}}}
     Scale Curve=&RandomRangeFloat(Minimum=0.1, Maximum=0.2)
    )
    Index 1 Param 3=&FloatFromCurve(FloatCurve=DataInterface() {propertyValues {Curve {keys [
         {interpMode=RCIM_Cubic, leaveTangent=6.3917, tangentMode=RCTM_Break, time=0, value=0}
         {arriveTangent=1.7549, interpMode=RCIM_Cubic, leaveTangent=10.6767, tangentMode=RCTM_User, time=0.25, value=1}
        ]}}})
    Index 1 Param 4=&FloatFromCurve(
     FloatCurve=DataInterface() {propertyValues {Curve {keys [{arriveTangent=-0.136, interpMode=RCIM_Cubic, leaveTangent=-0.136, tangentMode=RCTM_User, time=0, value=1}, <at 1 0.95>]}}}
     Scale Curve=2
    )
    Index 2 Param 1=&FloatFromCurve(
     FloatCurve=DataInterface() {propertyValues {Curve {keys [
         {interpMode=RCIM_Cubic, leaveTangent=0.4766, tangentMode=RCTM_Break, time=0, value=0}
         {arriveTangent=1.0548, interpMode=RCIM_Cubic, leaveTangent=1.0548, tangentMode=RCTM_User, time=0.5, value=0.45}
         <at 1 1>
        ]}}}
     Scale Curve=3
    )
    Index 2 Param 2=&FloatFromCurve(
     FloatCurve=DataInterface() {propertyValues {Curve {keys [{interpMode=RCIM_Cubic, tangentMode=RCTM_User, time=0, value=0}, {interpMode=RCIM_Cubic, tangentMode=RCTM_User, time=1, value=1}]}}}
     Scale Curve=0.3
    )
    Index 2 Param 3=&FloatFromCurve(
     FloatCurve=DataInterface() {propertyValues {Curve {keys [
         {interpMode=RCIM_Cubic, leaveTangent=-15.7052, leaveTangentWeight=0.1411, tangentMode=RCTM_Break, time=0, value=1}
         {interpMode=RCIM_Cubic, tangentMode=RCTM_Break, time=0.1, value=0.2}
         {arriveTangent=0.7532, interpMode=RCIM_Cubic, leaveTangent=0.7532, tangentMode=RCTM_User, time=0.5, value=0.45}
         {arriveTangent=0.4925, arriveTangentWeight=1.057, interpMode=RCIM_Cubic, tangentMode=RCTM_Break, time=1, value=0.75}
        ]}}}
     Scale Curve=5
    )
    Index 2 Param 4=&RandomRangeFloat(Minimum=1.5, Maximum=1.75)
    Module.Param0WriteEnabled=true
    Module.Param0WriteEnabled1=true
    Module.Param0WriteEnabled2=true
    Module.Param1WriteEnabled=false
    Module.Param1WriteEnabled1=true
    Module.Param1WriteEnabled2=true
    Module.Param2WriteEnabled=true
    Module.Param2WriteEnabled1=true
    Module.Param2WriteEnabled2=true
    Module.Param3WriteEnabled=true
    Module.Param3WriteEnabled1=true
    Module.Param3WriteEnabled2=true
    Write Parameter Index 1=true
    Write Parameter Index 2=true
   )
   &ScaleColor(
    Scale Mode=::NewEnumerator1
    Scale RGBA=/Niagara/DynamicInputs/TypeConversions/MakeCustomVector4FromBool(
     Bool:NiagaraBool=&Boolean_OrOperation(
      Bool1=&IsPlatformSetActive(New Platform Set=DataInterface() {propertyValues {Platforms {qualityLevelMask=-4}}})
      Bool2=$User.bIsFrontEndPreview
     )
     True Vector4:Vector4f=&MakeVector4_FromLinearColor(LinearColor=&FNTimeOfDayColor(
       Night Color=&RandomRangeLinearColor(Minimum=<rgba 0.5737 0.5737 0.6313 0.75>, Maximum=<rgba 0.9478 0.9478 1 0.75>)
       Morning Color=&RandomRangeLinearColor(Minimum=<rgba 0.418 0.4032 0.478 0.8589>, Maximum=<rgba 0.8207 0.8651 0.9253 0.75>)
       Day Color=&RandomRangeLinearColor(Minimum=<rgba 0.5261 0.5082 0.5032 0.5>, Maximum=<rgba 0.781 0.781 0.781 0.7>)
       Evening Color=&RandomRangeLinearColor(Minimum=<rgba 0.478 0.478 0.478 0.75>, Maximum=<rgba 0.781 0.781 0.781 0.75>)
      ))
     False Vector4:Vector4f=&MakeVector4_FromLinearColor(LinearColor=&FNTimeOfDayColor(
       Night Color=&RandomRangeLinearColor(Minimum=<rgba 0.4968 0.4968 0.6313 0.75>, Maximum=<rgba 0.7783 0.7783 1 0.75>)
       Morning Color=&RandomRangeLinearColor(Minimum=<rgba 0.2565 0.2764 0.3783 0.8589>, Maximum=<rgba 0.8087 0.8582 0.9253 0.75>)
       Day Color=&RandomRangeLinearColor(Minimum=<rgba 0.5261 0.4868 0.5652 0.5>, Maximum=<rgba 0.781 0.7233 0.7233 0.7>)
       Evening Color=&RandomRangeLinearColor(Minimum=<rgba 0.4261 0.4478 0.4136 0.75>, Maximum=<rgba 0.6783 0.6478 0.6282 0.75>)
      ))
    )
    ScaleRGBA=true
   ) {module=ScaleColor002}
   &ScaleColor(
    Scale Alpha=&FloatFromCurve(
     FloatCurve=DataInterface() {propertyValues {Curve {keys [
         {arriveTangent=2.6927, interpMode=RCIM_Cubic, leaveTangent=2.6927, tangentMode=RCTM_User, time=0, value=0}
         {arriveTangent=0.1984, leaveTangent=0.0132, tangentMode=RCTM_User, time=0.6, value=0.7}
         <at 2 0.6>
         {arriveTangent=-0.1351, interpMode=RCIM_Cubic, leaveTangent=-0.3618, tangentMode=RCTM_Break, time=3.2, value=0.4}
         {arriveTangent=-0.225, tangentMode=RCTM_User, time=5, value=0}
        ]}}}
     CurveIndex=$Emitter.Age
    )
    ScaleA=true
    ScaleRGB=false
   )
  ]
  rendererClasses [&NiagaraSpriteRendererProperties]
  renderers [
   {rendererClass=&NiagaraSpriteRendererProperties}
   {rendererClass=&NiagaraSpriteRendererProperties, rendererIndex=1}
  ]
 }
 Main_Meteor {
  emitterUpdateScript [
   &EmitterState(Reset Age On Awaken=true, Scalability Mode=::NewEnumerator1)
   &SpawnBurst_Instantaneous(Use Spawn Probability=false, UseLoopCountLimit=false) {module=SpawnBurst_Instantaneous001}
   &SpawnBurst_Instantaneous(Spawn Group=1, Use Spawn Probability=false, UseLoopCountLimit=false) {module=SpawnBurst_Instantaneous002}
   &SpawnBurst_Instantaneous(Spawn Group=2, Use Spawn Probability=false, UseLoopCountLimit=false) {module=SpawnBurst_Instantaneous003}
   &SpawnBurst_Instantaneous(Spawn Group=3, Use Spawn Probability=false, UseLoopCountLimit=false) {module=SpawnBurst_Instantaneous004}
   &SpawnBurst_Instantaneous(Spawn Group=4, Use Spawn Probability=false, UseLoopCountLimit=false) {module=SpawnBurst_Instantaneous005}
   &SpawnBurst_Instantaneous(Spawn Group=5, Use Spawn Probability=false, UseLoopCountLimit=false) {module=SpawnBurst_Instantaneous006}
   &SpawnBurst_Instantaneous(Spawn Group=6, Use Spawn Probability=false, UseLoopCountLimit=false) {module=SpawnBurst_Instantaneous007}
   &SpawnBurst_Instantaneous(Spawn Group=7, Use Spawn Probability=false, UseLoopCountLimit=false) {module=SpawnBurst_Instantaneous008}
   &SpawnBurst_Instantaneous(Spawn Group=8, Use Spawn Probability=false, UseLoopCountLimit=false)
  ]
  particleSpawnScript [
   &SkeletalMeshLocation(
    Bone / Socket Index=2
    MaskSpawnGroup=false
    Mesh Sampling Type=::NewEnumerator5
    Orientation Sampling=::NewEnumerator1
    Position Sampling=::NewEnumerator1
    Sampled Position Offset=&SelectVectorFromArray(
     Direct Array Index=&SpawnGroup()
     Vector Selection Array=DataInterface() {propertyValues {FloatData [
        <7.4773 2.124 12.6702>
        <8.7709 -25.7537 -4.0632>
        <-7.1841 -13.4425 5.4889>
        <-9.692 -23.0396 -11.5897>
        <11.0183 0.216 -10.1684>
        <4.2772 38.9138 13.3627>
        <-9.277 22.4491 2.4173>
        <-9.6866 -6.2382 -18.2187>
       ]}}
    )
    Sampled Position Offset Coordinate Space:ENiagaraSkelMeshTransforms=::NewEnumerator2
    Skeletal Mesh=$System.Skeletal Mesh
    Socket Sampling Mode=::NewEnumerator5
    Use Sampled Mesh Attribute as Ribbon ID=false
   ) {module=SkeletalMeshLocation004}
   &SkeletalMeshLocation(
    Bone / Socket Index=3
    MaskSpawnGroup=false
    Mesh Sampling Type=::NewEnumerator5
    Orientation Sampling=::NewEnumerator1
    Position Sampling=::NewEnumerator2
    Sampled Position Offset Coordinate Space:ENiagaraSkelMeshTransforms=::NewEnumerator2
    Skeletal Mesh=$System.Skeletal Mesh
    Socket Sampling Mode=::NewEnumerator5
    Use Sampled Mesh Attribute as Ribbon ID=false
   ) {module=SkeletalMeshLocation001}
   &SkeletalMeshLocation(
    MaskSpawnGroup=false
    Mesh Sampling Type=::NewEnumerator5
    Orientation Sampling=::NewEnumerator1
    Position Sampling=::NewEnumerator2
    Sampled Position Offset Coordinate Space:ENiagaraSkelMeshTransforms=::NewEnumerator2
    Skeletal Mesh=$System.Skeletal Mesh
    Socket Sampling Mode=::NewEnumerator5
    Use Sampled Mesh Attribute as Ribbon ID=false
   ) {module=SkeletalMeshLocation002}
   &InitializeParticle(
    Lifetime=&MaskFloatBySpawnGroup(Masked Float=1, Passthrough Float=5, Spawn Group Mask=8)
    Mesh Renderer Array Visibility Mode=::NewEnumerator1
    Mesh Renderer Info=DataInterface() {propertyValues {MeshRenderer=&self:Main_Meteor_0.NiagaraMeshRendererProperties_1}}
    Mesh Renderer Mesh Index:NiagaraInt32=&SpawnGroup()
    Mesh Scale Mode=::NewEnumerator3
    Position Offset=&SelectVectorFromArray(
     Direct Array Index=&SpawnGroup()
     Vector Selection Array=DataInterface() {propertyValues {FloatData [
        <7.4773 -2.124 12.6702>
        <8.7709 25.7537 -4.0632>
        <-7.1841 13.4425 5.4889>
        <-9.692 23.0396 -11.5897>
        <11.0183 -0.216 -10.1684>
        <4.2772 -38.9138 13.3627>
        <-9.277 -22.4491 2.4173>
        <-9.6866 6.2382 -18.2187>
       ]}}
    )
    Sprite Size Mode=::NewEnumerator3
    Uniform Sprite Size=3
    UsePositionOffset=false
    Write Material Random=true
   )
   &InitialMeshOrientation(Mesh Orientation Mode=::NewEnumerator7, Rotation=<0.05 0 0.05>, Use Rotation Vector=true)
   /SetVariables(Particles.VisibilityTag=&MaskIntBySpawnGroup(Masked Int=8, Passthrough Int=0, Spawn Group Mask=8)) {isSetParameters=true}
  ]
  particleUpdateScript [
   &self:BisBreakingScratchPad(
    BreakingPosition:NiagaraPosition=$Particles.SkeletalMeshLocation001.SampledPosition
    DistanceToBreakPos:NiagaraFloat=&FloatFromCurve(
     FloatCurve=DataInterface() {propertyValues {Curve {keys [<at 0.01 0>,<at 0.05 1>]}}}
     CurveIndex=$Emitter.LoopedAge
     Scale Curve=&MaskFloatBySpawnGroup(
      Masked Float=25
      Passthrough Float=&MaskFloatBySpawnGroup(Masked Float=25, Passthrough Float=20, Spawn Group Mask=1)
     )
    )
    Duration=2
    SetTrueAfterTime=0.75
   ) {module=BisBreakingScratchPad}
   &SkeletalMeshLocation(
    Bone / Socket Index=2
    "Interpolate Mesh Sampling (Update)"=true
    MaskSpawnGroup=false
    Mesh Sampling Type=::NewEnumerator5
    Orientation Sampling=::NewEnumerator1
    Sampled Orientation Coordinate Space=::Local
    Sampled Position Coordinate Space=::Local
    Sampled Position Offset=&SelectVectorFromArray(
     Direct Array Index=&SpawnGroup()
     Vector Selection Array=DataInterface() {propertyValues {FloatData [
        <-12.5227 2.8317 12.6702>
        <16.46 -16.8005 -11.7739>
        <-2.0959 -18.4014 0.285>
        <-2.2637 -17.9825 -19.1418>
        <10.1377 -5.9845 -9.5869>
        <-6.7602 34.9541 24.0537>
        <-15.2515 -1.0729 8.2977>
        <-8.3973 -1.0724 -19.9791>
        <0 0.3153 -0.1396>
       ]}}
    )
    Sampled Position Offset Coordinate Space:ENiagaraSkelMeshTransforms=::NewEnumerator2
    Sampled Velocity Coordinate Space:ENiagaraCoordinateSpace=::Local
    Skeletal Mesh=$System.Skeletal Mesh
    Socket Sampling Mode=::NewEnumerator5
    Use Sampled Mesh Attribute as Ribbon ID=false
    Write To Particle Position:NiagaraBool=&InvertBool(Boolean:NiagaraBool=$Particles.bIsBreaking)
   ) {module=SkeletalMeshLocation003}
   &SkeletalMeshLocation(
    Bone / Socket Index=&SpawnGroup()
    Bone Sampling Mode=::NewEnumerator4
    "Inherit Velocity (Scale)":NiagaraFloat=&FloatFromCurve(
     FloatCurve=DataInterface() {propertyValues {Curve {keys [
         {arriveTangent=0.1014, interpMode=RCIM_Cubic, leaveTangent=0.1014, tangentMode=RCTM_User, time=0, value=0}
         {arriveTangent=0.0852, interpMode=RCIM_Cubic, leaveTangent=0.0852, tangentMode=RCTM_User, time=2, value=0.35}
        ]}}}
     CurveIndex=$Particles.BisBreakingScratchPad.BooleanTimeline
     Scale Curve=&MakeCustomFloatFromBool(
      Bool=$Particles.bIsBreaking
      True Float=&MaskFloatBySpawnGroup(
       Masked Float=1
       Passthrough Float=&MaskFloatBySpawnGroup(
        Masked Float=1
        Passthrough Float=&MaskFloatBySpawnGroup(
         Masked Float=1
         Passthrough Float=&MaskFloatBySpawnGroup(
          Masked Float=1
          Passthrough Float=&MaskFloatBySpawnGroup(
           Masked Float=&FloatFromCurve(
            FloatCurve=DataInterface() {propertyValues {Curve {keys [
                <at 0 0.3 RCIM_Cubic>
                {arriveTangent=0.2573, interpMode=RCIM_Cubic, leaveTangent=0.2573, tangentMode=RCTM_User, time=0.8, value=0.4}
                <at 1 1 RCIM_Cubic>
               ]}}}
            CurveIndex=$Particles.BisBreakingScratchPad.BooleanTimeline
           )
           Passthrough Float=&MaskFloatBySpawnGroup(
            Masked Float=1
            Passthrough Float=&MaskFloatBySpawnGroup(
             Masked Float=1
             Passthrough Float=&MaskFloatBySpawnGroup(Masked Float=1, Passthrough Float=1, Spawn Group Mask=7)
             Spawn Group Mask=6
            )
            Spawn Group Mask=5
           )
           Spawn Group Mask=4
          )
          Spawn Group Mask=3
         )
         Spawn Group Mask=2
        )
        Spawn Group Mask=1
       )
      )
      False Float=0
     )
    )
    Initialize Particle Position To Sampled Position=false
    MaskSpawnGroup=false
    Orientation Sampling=::NewEnumerator1
    Position Sampling=::NewEnumerator1
    Sampled Orientation Coordinate Space=::Local
    Sampled Position Coordinate Space=::Local
    Sampled Position Offset Coordinate Space:ENiagaraSkelMeshTransforms=::NewEnumerator2
    Sampled Velocity Coordinate Space:ENiagaraCoordinateSpace=::Local
    Skeletal Mesh=$System.Skeletal Mesh
    Source Movement Scale:Vector3f=<0 0 0>
    Use Sampled Mesh Attribute as Ribbon ID=false
    Velocity Sampling=::NewEnumerator0
    VelocitySpeedLimit=false
    Write To Particle Position=false
   )
   &ParticleState()
   &ScaleMeshSize(Scale Factor=&MaskVectorBySpawnGroup(
     Masked Vector=&VectorFromFloat(Value=&FloatFromCurve(FloatCurve=DataInterface() {propertyValues {Curve {keys [<at 0 1>,<at 1 1>]}}}, CurveIndex=$Particles.BisBreakingScratchPad.BooleanTimeline))
     Passthrough Vector=&MaskVectorBySpawnGroup(
      Masked Vector=&VectorFromFloat(Value=&FloatFromCurve(
        FloatCurve=DataInterface() {propertyValues {Curve {keys [<at 0 1>, {arriveTangent=-2.1616, interpMode=RCIM_Cubic, leaveTangent=-2.1616, tangentMode=RCTM_User, time=0.4, value=1}, <at 1 0.6>]}}}
        CurveIndex=$Particles.BisBreakingScratchPad.BooleanTimeline
       ))
      Passthrough Vector=&MaskVectorBySpawnGroup(
       Masked Vector=&VectorFromFloat(Value=&FloatFromCurve(FloatCurve=DataInterface() {propertyValues {Curve {keys [<at 0 1>,<at 1 1>]}}}, CurveIndex=$Particles.BisBreakingScratchPad.BooleanTimeline))
       Passthrough Vector=&MaskVectorBySpawnGroup(
        Masked Vector=&VectorFromFloat(Value=&FloatFromCurve(
          FloatCurve=DataInterface() {propertyValues {Curve {keys [<at 0 1>,<at 1 0.75>]}}}
          CurveIndex=$Particles.BisBreakingScratchPad.BooleanTimeline
         ))
        Passthrough Vector=&MaskVectorBySpawnGroup(
         Masked Vector=&VectorFromFloat(Value=&FloatFromCurve(
           FloatCurve=DataInterface() {propertyValues {Curve {keys [<at 0 1>,<at 1 0.9>]}}}
           CurveIndex=$Particles.BisBreakingScratchPad.BooleanTimeline
          ))
         Passthrough Vector=&MaskVectorBySpawnGroup(
          Masked Vector=&VectorFromFloat(Value=&FloatFromCurve(FloatCurve=DataInterface() {propertyValues {Curve {keys [<at 0 1>,<at 1 1>]}}}, CurveIndex=$Particles.BisBreakingScratchPad.BooleanTimeline))
          Passthrough Vector=&MaskVectorBySpawnGroup(
           Masked Vector=&VectorFromFloat(Value=&FloatFromCurve(
             FloatCurve=DataInterface() {propertyValues {Curve {keys [<at 0 1>,<at 1 0.75>]}}}
             CurveIndex=$Particles.BisBreakingScratchPad.BooleanTimeline
            ))
           Passthrough Vector=&MaskVectorBySpawnGroup(
            Masked Vector=&VectorFromFloat(Value=&FloatFromCurve(
              FloatCurve=DataInterface() {propertyValues {Curve {keys [<at 0 1>,<at 1 0.75>]}}}
              CurveIndex=$Particles.BisBreakingScratchPad.BooleanTimeline
             ))
            Passthrough Vector=&VectorFromFloat(Value=&FloatFromCurve(
              FloatCurve=DataInterface() {propertyValues {Curve {keys [<at 0 1>,<at 1 0.75>]}}}
              CurveIndex=$Particles.BisBreakingScratchPad.BooleanTimeline
             ))
            Spawn Group Mask=7
           )
           Spawn Group Mask=6
          )
          Spawn Group Mask=5
         )
         Spawn Group Mask=4
        )
        Spawn Group Mask=3
       )
       Spawn Group Mask=2
      )
      Spawn Group Mask=1
     )
    )) {module=ScaleMeshSize002}
   &ScaleMeshSize(Scale Factor=&MakeCustomVectorFromBool(
     Bool=$Particles.bIsBreaking
     True Vector=&VectorFromCurve(
      VectorCurve=DataInterface() {propertyValues {
        XCurve {keys [<at 0 1>, {arriveTangent=-1, interpMode=RCIM_Cubic, leaveTangent=-1, time=0.2, value=1}, <at 1 0 RCIM_Cubic>]}
        YCurve {keys [<at 0 1>, {arriveTangent=-1, interpMode=RCIM_Cubic, leaveTangent=-1, time=0.2, value=1}, <at 1 0 RCIM_Cubic>]}
        ZCurve {keys [<at 0 1>, {arriveTangent=-1, interpMode=RCIM_Cubic, leaveTangent=-1, time=0.2, value=1}, <at 1 0 RCIM_Cubic>]}
       }}
      CurveIndex=&NormalizeDistanceRange(
       Start Position:NiagaraPosition=$Engine.Emitter.SimulationPosition
       End Position:NiagaraPosition=$Particles.Position
       Distance:NiagaraFloat=&FloatFromCurve(FloatCurve=DataInterface() {propertyValues {Curve {keys [<at 0 1200>,<at 2 400>]}}}, CurveIndex=$Emitter.LoopedAge)
       Invert Normalized Range:NiagaraBool=false
      )
     )
     False Vector=<1 1 1>
    )) {module=ScaleMeshSize001}
   &ScaleMeshSize(Scale Factor=&VectorFromCurve(
     VectorCurve=DataInterface() {propertyValues {
       XCurve {keys [<at 0.05 1>, {arriveTangent=-0.101, interpMode=RCIM_Cubic, leaveTangent=-0.101, time=1, value=1}, <at 5 0.5>]}
       YCurve {keys [<at 0.05 1>, {arriveTangent=-0.101, interpMode=RCIM_Cubic, leaveTangent=-0.101, time=1, value=1}, <at 5 0.5>]}
       ZCurve {keys [<at 0.05 1>, {arriveTangent=-0.101, interpMode=RCIM_Cubic, leaveTangent=-0.101, time=1, value=1}, <at 5 0.5>]}
      }}
     CurveIndex=$Emitter.Age
    ))
   &PointForce(
    Force Falloff Distance=50
    Force Origin:NiagaraPosition=$Particles.SkeletalMeshLocation001.SampledPosition
    Force Strength=&FloatFromCurve(
     FloatCurve=DataInterface() {propertyValues {Curve {keys [
         {arriveTangent=-0.2018, interpMode=RCIM_Cubic, leaveTangent=-0.2018, tangentMode=RCTM_User, time=0.3, value=2}
         {arriveTangent=0.3383, tangentMode=RCTM_User, time=0.9, value=0.125}
        ]}}}
     CurveIndex=$Emitter.LoopedAge
     Scale Curve=&Multiply_Float(A=10, B=1)
    )
    Linear Force Vector:Vector3f=&MaskVectorBySpawnGroup(
     Masked Vector=&MakeVector(
      X=&FloatFromCurve(
       FloatCurve=DataInterface() {propertyValues {Curve {
          defaultValue=0
          keys [{interpMode=RCIM_Cubic, tangentMode=RCTM_User, time=0, value=0}, {interpMode=RCIM_Cubic, tangentMode=RCTM_User, time=1, value=1}]
         }}}
       CurveIndex=$Particles.BisBreakingScratchPad.BooleanTimeline
       Scale Curve=&Multiply_Float(A=50, B=&MakeCustomFloatFromBool(Bool=$Particles.bIsBreaking, True Float=1, False Float=0))
      )
      Y=&FloatFromCurve(
       FloatCurve=DataInterface() {propertyValues {Curve {
          defaultValue=0
          keys [
           {interpMode=RCIM_Cubic, tangentMode=RCTM_User, time=0, value=-0.2}
           {arriveTangent=1.2, interpMode=RCIM_Cubic, leaveTangent=1.2, time=0.05, value=-0.05}
           {arriveTangent=1.3398, interpMode=RCIM_Cubic, leaveTangent=1.3398, tangentMode=RCTM_User, time=1, value=1}
          ]
         }}}
       CurveIndex=$Particles.BisBreakingScratchPad.BooleanTimeline
       Scale Curve=&Multiply_Float(A=-500, B=&MakeCustomFloatFromBool(Bool=$Particles.bIsBreaking, True Float=1, False Float=0))
      )
      Z=&FloatFromCurve(
       FloatCurve=DataInterface() {propertyValues {Curve {
          defaultValue=0
          keys [{interpMode=RCIM_Cubic, tangentMode=RCTM_User, time=0, value=0}, {interpMode=RCIM_Cubic, tangentMode=RCTM_User, time=1, value=1}]
         }}}
       CurveIndex=$Particles.BisBreakingScratchPad.BooleanTimeline
       Scale Curve=&Multiply_Float(A=50, B=&MakeCustomFloatFromBool(Bool=$Particles.bIsBreaking, True Float=1, False Float=0))
      )
     )
     Passthrough Vector=&MaskVectorBySpawnGroup(
      Masked Vector=&MakeVector(
       X=&FloatFromCurve(
        FloatCurve=DataInterface() {propertyValues {Curve {
           defaultValue=0
           keys [{interpMode=RCIM_Cubic, tangentMode=RCTM_User, time=0, value=0}, {interpMode=RCIM_Cubic, tangentMode=RCTM_User, time=1, value=1}]
          }}}
        CurveIndex=$Particles.BisBreakingScratchPad.BooleanTimeline
        Scale Curve=&Multiply_Float(A=-50, B=&MakeCustomFloatFromBool(Bool=$Particles.bIsBreaking, True Float=1, False Float=0))
       )
       Y=0
       Z=&FloatFromCurve(
        FloatCurve=DataInterface() {propertyValues {Curve {
           defaultValue=0
           keys [
            {arriveTangent=-2.2876, interpMode=RCIM_Cubic, leaveTangent=-2.2876, tangentMode=RCTM_User, time=0, value=0}
            {interpMode=RCIM_Cubic, tangentMode=RCTM_User, time=1, value=1}
           ]
          }}}
        CurveIndex=$Particles.BisBreakingScratchPad.BooleanTimeline
        Scale Curve=&Multiply_Float(A=50, B=&MakeCustomFloatFromBool(Bool=$Particles.bIsBreaking, True Float=1, False Float=0))
       )
      )
      Passthrough Vector=&MaskVectorBySpawnGroup(
       Masked Vector=&MakeVector(
        X=&FloatFromCurve(
         FloatCurve=DataInterface() {propertyValues {Curve {
            defaultValue=1
            keys [{interpMode=RCIM_Cubic, tangentMode=RCTM_User, time=0, value=0}, {interpMode=RCIM_Cubic, tangentMode=RCTM_User, time=1, value=1}]
           }}}
         CurveIndex=$Particles.BisBreakingScratchPad.BooleanTimeline
         Scale Curve=&Multiply_Float(A=-100, B=&MakeCustomFloatFromBool(Bool=$Particles.bIsBreaking, True Float=1, False Float=0))
        )
        Y=0
        Z=&FloatFromCurve(
         FloatCurve=DataInterface() {propertyValues {Curve {
            defaultValue=1
            keys [{interpMode=RCIM_Cubic, tangentMode=RCTM_User, time=0, value=0}, {interpMode=RCIM_Cubic, tangentMode=RCTM_User, time=1, value=1}]
           }}}
         CurveIndex=$Particles.BisBreakingScratchPad.BooleanTimeline
         Scale Curve=&Multiply_Float(A=30, B=&MakeCustomFloatFromBool(Bool=$Particles.bIsBreaking, True Float=1, False Float=0))
        )
       )
       Passthrough Vector=&MaskVectorBySpawnGroup(
        Masked Vector=&MakeVector(
         X=&FloatFromCurve(
          FloatCurve=DataInterface() {propertyValues {Curve {keys [
              <at 0 0 RCIM_Cubic>
              {arriveTangent=0.0688, interpMode=RCIM_Cubic, leaveTangent=0.0688, time=1.4, value=0.45}
              {interpMode=RCIM_Cubic, leaveTangent=0.0437, time=2, value=0.1375}
             ]}}}
          CurveIndex=$Particles.BisBreakingScratchPad.BooleanTimeline
          Scale Curve=&Multiply_Float(A=10, B=&MakeCustomFloatFromBool(Bool=$Particles.bIsBreaking, True Float=1, False Float=0))
         )
         Y=0
         Z=0
        )
        Passthrough Vector=&MaskVectorBySpawnGroup(
         Masked Vector=&MakeVector(
          X=&FloatFromCurve(
           FloatCurve=DataInterface() {propertyValues {Curve {keys [
               <at 0 0 RCIM_Cubic>
               {arriveTangent=0.0687, interpMode=RCIM_Cubic, leaveTangent=0.0687, time=0.8, value=0.45}
               {interpMode=RCIM_Cubic, leaveTangent=0.0437, time=2, value=0.1375}
              ]}}}
           CurveIndex=$Particles.BisBreakingScratchPad.BooleanTimeline
           Scale Curve=&Multiply_Float(A=100, B=&MakeCustomFloatFromBool(Bool=$Particles.bIsBreaking, True Float=1, False Float=0))
          )
          Y=&FloatFromCurve(
           FloatCurve=DataInterface() {propertyValues {Curve {keys [
               <at 0 0 RCIM_Cubic>
               {arriveTangent=0.0688, interpMode=RCIM_Cubic, leaveTangent=0.0688, time=1.4, value=0.45}
               {interpMode=RCIM_Cubic, leaveTangent=0.0437, time=2, value=0.1375}
              ]}}}
           CurveIndex=$Particles.BisBreakingScratchPad.BooleanTimeline
           Scale Curve=&Multiply_Float(A=180, B=&MakeCustomFloatFromBool(Bool=$Particles.bIsBreaking, True Float=1, False Float=0))
          )
          Z=&FloatFromCurve(
           FloatCurve=DataInterface() {propertyValues {Curve {keys [
               <at 0 0 RCIM_Cubic>
               {arriveTangent=0.0687, interpMode=RCIM_Cubic, leaveTangent=0.0687, time=0.8, value=0.5}
               {interpMode=RCIM_Cubic, leaveTangent=0.0437, time=2, value=0.1375}
              ]}}}
           CurveIndex=$Particles.BisBreakingScratchPad.BooleanTimeline
           Scale Curve=&Multiply_Float(A=-95, B=&MakeCustomFloatFromBool(Bool=$Particles.bIsBreaking, True Float=1, False Float=0))
          )
         )
         Passthrough Vector=&MaskVectorBySpawnGroup(
          Masked Vector=&MakeVector(
           X=&FloatFromCurve(
            FloatCurve=DataInterface() {propertyValues {Curve {keys [
                <at 0 0 RCIM_Cubic>
                {arriveTangent=0.0688, interpMode=RCIM_Cubic, leaveTangent=0.0688, time=0.8, value=0.45}
                {interpMode=RCIM_Cubic, leaveTangent=0.0437, time=2, value=0.1375}
               ]}}}
            CurveIndex=$Particles.BisBreakingScratchPad.BooleanTimeline
            Scale Curve=&Multiply_Float(A=30, B=&MakeCustomFloatFromBool(Bool=$Particles.bIsBreaking, True Float=1, False Float=0))
           )
           Y=0
           Z=&FloatFromCurve(
            FloatCurve=DataInterface() {propertyValues {Curve {keys [
                <at 0 0 RCIM_Cubic>
                {arriveTangent=0.0688, interpMode=RCIM_Cubic, leaveTangent=0.0688, time=0.8, value=0.45}
                {interpMode=RCIM_Cubic, leaveTangent=0.0437, time=2, value=0.1375}
               ]}}}
            CurveIndex=$Particles.BisBreakingScratchPad.BooleanTimeline
            Scale Curve=&Multiply_Float(A=-500, B=&MakeCustomFloatFromBool(Bool=$Particles.bIsBreaking, True Float=1, False Float=0))
           )
          )
          Passthrough Vector=&MaskVectorBySpawnGroup(
           Masked Vector=&MakeVector(
            X=&FloatFromCurve(
             FloatCurve=DataInterface() {propertyValues {Curve {keys [
                 <at 0 0.475>
                 {arriveTangent=-0.2812, interpMode=RCIM_Cubic, leaveTangent=-0.2812, time=1, value=0.475}
                 {interpMode=RCIM_Cubic, leaveTangent=0.0437, time=1.2, value=0.1375}
                ]}}}
             CurveIndex=$Particles.BisBreakingScratchPad.BooleanTimeline
             Scale Curve=&Multiply_Float(A=20, B=&MakeCustomFloatFromBool(Bool=$Particles.bIsBreaking, True Float=1, False Float=0))
            )
            Y=0
            Z=&FloatFromCurve(
             FloatCurve=DataInterface() {propertyValues {Curve {keys [
                 <at 0 0 RCIM_Cubic>
                 {arriveTangent=0.0688, interpMode=RCIM_Cubic, leaveTangent=0.0688, time=0.8, value=0.45}
                 {interpMode=RCIM_Cubic, leaveTangent=0.0437, time=2, value=0.1375}
                ]}}}
             CurveIndex=$Particles.BisBreakingScratchPad.BooleanTimeline
             Scale Curve=&Multiply_Float(A=100, B=&MakeCustomFloatFromBool(Bool=$Particles.bIsBreaking, True Float=1, False Float=0))
            )
           )
           Passthrough Vector=&MaskVectorBySpawnGroup(
            Masked Vector=&MakeVector(
             X=&FloatFromCurve(
              FloatCurve=DataInterface() {propertyValues {Curve {keys [
                  <at 0 0.475>
                  {arriveTangent=-0.2812, interpMode=RCIM_Cubic, leaveTangent=-0.2812, time=1, value=0.475}
                  {interpMode=RCIM_Cubic, leaveTangent=0.0437, time=1.2, value=0.1375}
                 ]}}}
              CurveIndex=$Particles.BisBreakingScratchPad.BooleanTimeline
              Scale Curve=&Multiply_Float(A=10, B=&MakeCustomFloatFromBool(Bool=$Particles.bIsBreaking, True Float=1, False Float=0))
             )
             Y=0
             Z=&FloatFromCurve(
              FloatCurve=DataInterface() {propertyValues {Curve {keys [
                  <at 0 0.475>
                  {arriveTangent=-0.4821, interpMode=RCIM_Cubic, leaveTangent=-0.4821, time=0.5, value=0.475}
                  {interpMode=RCIM_Cubic, leaveTangent=0.0437, time=0.7, value=0.1375}
                 ]}}}
              CurveIndex=$Particles.BisBreakingScratchPad.BooleanTimeline
              Scale Curve=&Multiply_Float(A=-5, B=&MakeCustomFloatFromBool(Bool=$Particles.bIsBreaking, True Float=1, False Float=0))
             )
            )
            Passthrough Vector=<0 0 0>
            Spawn Group Mask=7
           )
           Spawn Group Mask=6
          )
          Spawn Group Mask=5
         )
         Spawn Group Mask=4
        )
        Spawn Group Mask=3
       )
       Spawn Group Mask=2
      )
      Spawn Group Mask=1
     )
    )
    Module.Use Falloff=true
    Module.UseLinearForce=true
   ) {module=PointForce001}
   &Drag(
    Drag=&FloatFromCurve(FloatCurve=DataInterface() {propertyValues {Curve {keys [<at 0 8 RCIM_Constant>,<at 2.4 3 RCIM_Constant>]}}}, CurveIndex=$Emitter.Age)
    Use Linear Drag=true
    Use Rotational Drag=false
   ) {module=Drag001}
   /SetVariables(Particles.Velocity=&Lerp_Vector(
     A=<0 0 0>
     B=$Particles.Velocity
     Alpha=&ClampFloat(Float=&Multiply_Float(A=10, B=$Particles.BisBreakingScratchPad.BooleanTimeline))
    )) {isSetParameters=true}
   &MeshRotationForce(Rotation=&Multiply_VectorByFloat(
     Vector=&MaskVectorBySpawnGroup(
      Masked Vector=<1.5 1 -1>
      Passthrough Vector=&MaskVectorBySpawnGroup(
       Masked Vector=<1 2 1>
       Passthrough Vector=&MaskVectorBySpawnGroup(
        Masked Vector=<2 -1.5 1>
        Passthrough Vector=&MaskVectorBySpawnGroup(
         Masked Vector=<1 -1 1>
         Passthrough Vector=&MaskVectorBySpawnGroup(
          Masked Vector=<-1 1.25 -2>
          Passthrough Vector=&MaskVectorBySpawnGroup(
           Masked Vector=<1 1 -1>
           Passthrough Vector=&MaskVectorBySpawnGroup(
            Masked Vector=<-0.5 -1 1>
            Passthrough Vector=&MaskVectorBySpawnGroup(Masked Vector=<1 -1 2>, Passthrough Vector=<0 0 0>, Spawn Group Mask=7)
            Spawn Group Mask=6
           )
           Spawn Group Mask=5
          )
          Spawn Group Mask=4
         )
         Spawn Group Mask=3
        )
        Spawn Group Mask=2
       )
       Spawn Group Mask=1
      )
     )
     Float=&FloatFromCurve(
      FloatCurve=DataInterface() {propertyValues {Curve {keys [<at 0 0.25 RCIM_Cubic>, {arriveTangent=0.0155, interpMode=RCIM_Constant, time=2.2, value=0.25}, <at 2.3 2.5 RCIM_Cubic>]}}}
      CurveIndex=$Emitter.Age
     )
    ))
   &SolveForcesAndVelocity(Clamp Velocity=false, Limit Acceleration=false, Write to Presolve Properties=false)
   &ScaleColor(
    Scale Alpha=&MaskFloatBySpawnGroup(
     Masked Float=&FloatFromCurve(
      FloatCurve=DataInterface() {propertyValues {Curve {
         defaultValue=1
         keys [{interpMode=RCIM_Cubic, tangentMode=RCTM_User, time=0, value=1}, {interpMode=RCIM_Cubic, tangentMode=RCTM_User, time=1, value=0}]
        }}}
      Scale Curve=&MakeCustomFloatFromBool(
       Bool=&IsPlatformSetActive(New Platform Set=DataInterface() {propertyValues {Platforms {qualityLevelMask=-4}}})
       True Float=1
       False Float=0.2
      )
     )
     Passthrough Float=1
     Spawn Group Mask=8
    )
    ScaleA=true
    ScaleRGB=true
   )
   &DynamicMaterialParameters(
    Index 0 Param 1=1
    Index 0 Param 2=1
    Index 0 Param 3=1
    Index 1 Param 1=1
    Module.Index0Float0Write=true
    Module.Index0Float1Write=true
    Module.Index0Float2Write=true
    Module.Index0Float3Write=true
    Module.Index1Float0Write=true
    Module.Index1Float1Write=true
    Module.Index1Float2Write=true
    Module.Index1Float3Write=true
    Write Parameter Index 1=::NewEnumerator1
   )
  ]
  rendererClasses [&NiagaraMeshRendererProperties]
  renderers [
   {rendererClass=&NiagaraMeshRendererProperties, rendererIndex=0}
   {rendererClass=&NiagaraMeshRendererProperties, rendererIndex=1}
  ]
 }
 SmokeBIG {
  emitterUpdateScript [
   &EmitterState(Life Cycle Mode=::NewEnumerator1, Loop Behavior=::NewEnumerator1, Loop Delay=0.3, Loop Duration=1, UseLoopDelay=true)
   &SpawnBurst_Instantaneous(Spawn Count=16, Use Spawn Probability=false, UseLoopCountLimit=true)
   &SpawnBurst_Instantaneous(Spawn Group=1, Use Spawn Probability=false, UseLoopCountLimit=false) {module=SpawnBurst_Instantaneous001}
  ]
  particleSpawnScript [
   &InitializeParticle(
    Color Maximum=<rgba 0.3854 0.3854 0.3854 0.4957>
    Color Minimum=<rgba 0.2552 0.2552 0.2552 1.0412>
    Color Mode=::NewEnumerator2
    Lifetime Max=7.5
    Lifetime Min=2.5
    Lifetime Mode=::NewEnumerator1
    Sprite Rotation Mode=::NewEnumerator2
    Sprite Size=&MaskVector2dBySpawnGroup(
     Masked Vector2D=&RandomRangeVector2D(Minimum=<247.401 233.8594>, Maximum=<309.1217 334.8896>)
     Passthrough Vector2D=&RandomRangeVector2D(Minimum=<500 500>, Maximum=<610.2156 660.7867>)
     Spawn Group Mask=0
    )
    Sprite Size Mode=::NewEnumerator4
    Sprite UV Mode=::NewEnumerator3
    UsePositionOffset=false
    Write Material Random=true
   )
   &CameraOffset(Camera Offset Amount=50)
   &ShapeLocation(
    Handle Radius=5
    Large Radius=25.3192
    MaskSpawnGroup=true
    Offset=<0 0 -30>
    Offset Mode=::NewEnumerator0
    Override Seed=false
    Rotation Axis=<0 0 1>
    Rotation Mode=::NewEnumerator1
    Shape Primitive=::NewEnumerator3
    UseOwnerScale=true
    V Distribution=1
   )
   &ShapeLocation(MaskSpawnGroup=true, Override Seed=false, Spawn Group Mask=1, Sphere Radius=25, UseOwnerScale=true) {module=ShapeLocation001}
   &AddVelocity(
    Module.Use Falloff Distance=false
    Override Seed=false
    Velocity Mode=::NewEnumerator1
    Velocity Speed=&RandomRangeFloat(Minimum=1495.1179, Maximum=1519.1489)
   )
   &SubUVAnimation(Evaluation Type=::NewEnumerator0, SubUV Animation Mode:ENiagara_SubUVLookupModeV2=::NewEnumerator2, UseEndFrame=false, UseStartFrame=false)
  ]
  particleUpdateScript [
   &Drag(Drag=&RandomRangeFloat(Minimum=6, Maximum=11), Use Linear Drag=true, Use Rotational Drag=true)
   &AccelerationForce(Acceleration=&RandomRangeVector(Minimum=<0 0 -100>, Maximum=<0 0 200>))
   &SolveForcesAndVelocity(Clamp Velocity=false, Limit Acceleration=false)
   &ParticleState()
   &ScaleSpriteSize(Uniform Curve Sprite Scale=DataInterface() {propertyValues {Curve {keys [
        {arriveTangent=4.1286, interpMode=RCIM_Cubic, leaveTangent=4.1286, tangentMode=RCTM_User, time=0, value=0.475}
        {arriveTangent=1.2171, interpMode=RCIM_Cubic, leaveTangent=1.2171, tangentMode=RCTM_User, time=0.3, value=1.125}
        <at 1 1.4 RCIM_Cubic>
       ]}}})
   /SetVariables(Particles.SubImageIndex=&FloatFromCurve(
     FloatCurve=DataInterface() {propertyValues {Curve {keys [
         {arriveTangent=2.5597, interpMode=RCIM_Cubic, leaveTangent=2.5597, tangentMode=RCTM_User, time=0, value=0}
         {arriveTangent=1, interpMode=RCIM_Cubic, leaveTangent=1, time=0.1, value=0.75}
         {interpMode=RCIM_Cubic, leaveTangent=1.1425, time=1, value=1}
        ]}}}
     Scale Curve=&MakeCustomFloatFromBool(
      Bool=&IsPlatformSetActive(New Platform Set=DataInterface() {propertyValues {Platforms {qualityLevelMask=-4}}})
      True Float=64
      False Float=15
     )
    )) {isSetParameters=true}
   &DynamicMaterialParameters(
    Index 0 Param 1=&FloatFromCurve(FloatCurve=DataInterface() {propertyValues {Curve {keys [{arriveTangent=3.9017, interpMode=RCIM_Cubic, leaveTangent=3.9017, tangentMode=RCTM_User, time=0, value=0.075}, <at 0.5 1 RCIM_Cubic>]}}})
    Index 0 Param 4=1
    Module.Index0Float0Write=true
    Module.Index0Float1Write=true
    Module.Index0Float2Write=true
    Module.Index0Float3Write=true
   )
   &ScaleColor(
    Scale Mode=::NewEnumerator1
    Scale RGBA=&MakeCustomVector4FromBool(
     Bool:NiagaraBool=&Boolean_OrOperation(
      Bool1=&IsPlatformSetActive(New Platform Set=DataInterface() {propertyValues {Platforms {qualityLevelMask=-4}}})
      Bool2=$User.bIsFrontEndPreview
     )
     True Vector4:Vector4f=&MakeVector4_FromLinearColor(LinearColor=&FNTimeOfDayColor(
       Night Color=&RandomRangeLinearColor(Minimum=<rgba 0.5407 0.5407 0.6313 0.75>, Maximum=<rgba 0.8435 0.8435 1 0.75>)
       Morning Color=&RandomRangeLinearColor(Minimum=<rgba 0.418 0.4032 0.478 0.8589>, Maximum=<rgba 0.8207 0.8651 0.9253 0.75>)
       Day Color=&RandomRangeLinearColor(Minimum=<rgba 0.5261 0.5082 0.5032 0.5>, Maximum=<rgba 0.781 0.781 0.781 0.7>)
       Evening Color=&RandomRangeLinearColor(Minimum=<rgba 0.478 0.478 0.478 0.75>, Maximum=<rgba 0.781 0.781 0.781 0.75>)
      ))
     False Vector4:Vector4f=&MakeVector4_FromLinearColor(LinearColor=&FNTimeOfDayColor(
       Night Color=&RandomRangeLinearColor(Minimum=<rgba 0.3815 0.3815 0.6313 0.75>, Maximum=<rgba 0.5522 0.5522 1 0.75>)
       Morning Color=&RandomRangeLinearColor(Minimum=<rgba 0.2565 0.2764 0.3783 0.8589>, Maximum=<rgba 0.8087 0.8582 0.9253 0.75>)
       Day Color=&RandomRangeLinearColor(Minimum=<rgba 0.5261 0.4868 0.5652 0.5>, Maximum=<rgba 0.781 0.7233 0.7233 0.7>)
       Evening Color=&RandomRangeLinearColor(Minimum=<rgba 0.4261 0.4478 0.4136 0.75>, Maximum=<rgba 0.6783 0.6478 0.6282 0.75>)
      ))
    )
    ScaleRGBA=true
   ) {module=ScaleColor002}
   &ScaleColor(
    Scale Alpha=&MaskFloatBySpawnGroup(
     Masked Float=&FloatFromCurve(FloatCurve=DataInterface() {propertyValues {Curve {keys [{arriveTangent=6.7279, interpMode=RCIM_Cubic, time=0, value=1}, {interpMode=RCIM_Cubic, leaveTangent=-0.9746, time=1, value=0}]}}})
     Passthrough Float=&FloatFromCurve(FloatCurve=DataInterface() {propertyValues {Curve {keys [
          {arriveTangent=-2.0221, interpMode=RCIM_Cubic, leaveTangent=-2.0221, tangentMode=RCTM_User, time=0, value=1}
          {interpMode=RCIM_Cubic, leaveTangent=-0.9746, time=1, value=0}
         ]}}})
    )
    Scale RGB=&MakeVectorFromLinearColorRGB(Color=&ColorFromCurve(ColorCurve=DataInterface() {propertyValues {
        AlphaCurve {keys []}
        BlueCurve {keys [{arriveTangent=-9.5363, interpMode=RCIM_Cubic, leaveTangent=-9.5363, tangentMode=RCTM_User, time=0, value=4}, <at 1 1 RCIM_Cubic>]}
        GreenCurve {keys [{arriveTangent=-9.5363, interpMode=RCIM_Cubic, leaveTangent=-9.5363, tangentMode=RCTM_User, time=0, value=4}, <at 1 1 RCIM_Cubic>]}
        RedCurve {keys [{arriveTangent=-9.5363, interpMode=RCIM_Cubic, leaveTangent=-9.5363, tangentMode=RCTM_User, time=0, value=4}, <at 1 1 RCIM_Cubic>]}
       }}))
    ScaleA=true
    ScaleRGB=true
   )
  ]
  rendererClasses [&NiagaraSpriteRendererProperties]
  renderers [{rendererClass=&NiagaraSpriteRendererProperties}]
 }
 SmokeBIGLOWMID {
  emitterUpdateScript [
   &EmitterState(Life Cycle Mode=::NewEnumerator1, Loop Behavior=::NewEnumerator1, Loop Delay=0.3, Loop Duration=1, UseLoopDelay=true)
   &SpawnBurst_Instantaneous(Spawn Count=16, Use Spawn Probability=false, UseLoopCountLimit=true)
   &SpawnBurst_Instantaneous(Spawn Group=1, Use Spawn Probability=false, UseLoopCountLimit=false) {module=SpawnBurst_Instantaneous001}
  ]
  particleSpawnScript [
   &InitializeParticle(
    Color Maximum=<rgba 0.3854 0.3854 0.3854 0.4957>
    Color Minimum=<rgba 0.2552 0.2552 0.2552 1.0412>
    Color Mode=::NewEnumerator2
    Lifetime Max=7.5
    Lifetime Min=2.5
    Lifetime Mode=::NewEnumerator1
    Sprite Rotation Mode=::NewEnumerator2
    Sprite Size=&MaskVector2dBySpawnGroup(
     Masked Vector2D=&RandomRangeVector2D(Minimum=<247.401 233.8594>, Maximum=<309.1217 334.8896>)
     Passthrough Vector2D=&RandomRangeVector2D(Minimum=<250 250>, Maximum=<305.1078 330.3933>)
     Spawn Group Mask=0
    )
    Sprite Size Mode=::NewEnumerator4
    Sprite UV Mode=::NewEnumerator3
    UsePositionOffset=false
    Write Material Random=true
   )
   &CameraOffset(Camera Offset Amount=50)
   &ShapeLocation(
    Handle Radius=5
    Large Radius=25.3192
    MaskSpawnGroup=true
    Offset=<0 0 -30>
    Offset Mode=::NewEnumerator0
    Override Seed=false
    Rotation Axis=<0 0 1>
    Rotation Mode=::NewEnumerator1
    Shape Primitive=::NewEnumerator3
    UseOwnerScale=true
    V Distribution=1
   )
   &ShapeLocation(MaskSpawnGroup=true, Override Seed=false, Spawn Group Mask=1, Sphere Radius=25, UseOwnerScale=true) {module=ShapeLocation001}
   &AddVelocity(
    Module.Use Falloff Distance=false
    Override Seed=false
    Velocity Mode=::NewEnumerator1
    Velocity Speed=&RandomRangeFloat(Minimum=1495.1179, Maximum=1519.1489)
   )
   &SubUVAnimation(
    Evaluation Type=::NewEnumerator0
    "Sprite Renderer "=DataInterface() {propertyValues {SpriteRenderer=&self:SmokeBIGLOWMID_0.NiagaraSpriteRendererProperties_2}}
    SubUV Animation Mode:ENiagara_SubUVLookupModeV2=::NewEnumerator2
    UseEndFrame=false
    UseStartFrame=false
   )
  ]
  particleUpdateScript [
   &Drag(Drag=&RandomRangeFloat(Minimum=6, Maximum=11), Use Linear Drag=true, Use Rotational Drag=true)
   &AccelerationForce(Acceleration=&RandomRangeVector(Minimum=<0 0 -100>, Maximum=<0 0 200>))
   &SolveForcesAndVelocity(Clamp Velocity=false, Limit Acceleration=false)
   &ParticleState()
   &ScaleSpriteSize(Uniform Curve Sprite Scale=DataInterface() {propertyValues {Curve {keys [
        {arriveTangent=4.1286, interpMode=RCIM_Cubic, leaveTangent=4.1286, tangentMode=RCTM_User, time=0, value=0.475}
        {arriveTangent=1.2171, interpMode=RCIM_Cubic, leaveTangent=1.2171, tangentMode=RCTM_User, time=0.3, value=1.125}
        <at 1 1.4 RCIM_Cubic>
       ]}}})
   &DynamicMaterialParameters(
    Index 0 Param 1=&FloatFromCurve(FloatCurve=DataInterface() {propertyValues {Curve {keys [{arriveTangent=3.9017, interpMode=RCIM_Cubic, leaveTangent=3.9017, tangentMode=RCTM_User, time=0, value=0.075}, <at 0.5 1 RCIM_Cubic>]}}})
    Index 0 Param 4=1
    Module.Index0Float0Write=true
    Module.Index0Float1Write=true
    Module.Index0Float2Write=true
    Module.Index0Float3Write=true
   ) {enabled=false}
   &ScaleColor(
    Scale Mode=::NewEnumerator1
    Scale RGBA=&MakeCustomVector4FromBool(
     Bool:NiagaraBool=&Boolean_OrOperation(
      Bool1=&IsPlatformSetActive(New Platform Set=DataInterface() {propertyValues {Platforms {qualityLevelMask=-4}}})
      Bool2=$User.bIsFrontEndPreview
     )
     True Vector4:Vector4f=&MakeVector4_FromLinearColor(LinearColor=&FNTimeOfDayColor(
       Night Color=&RandomRangeLinearColor(Minimum=<rgba 0.5407 0.5407 0.6313 0.75>, Maximum=<rgba 0.8435 0.8435 1 0.75>)
       Morning Color=&RandomRangeLinearColor(Minimum=<rgba 0.418 0.4032 0.478 0.8589>, Maximum=<rgba 0.8207 0.8651 0.9253 0.75>)
       Day Color=&RandomRangeLinearColor(Minimum=<rgba 0.5261 0.5082 0.5032 0.5>, Maximum=<rgba 0.781 0.781 0.781 0.7>)
       Evening Color=&RandomRangeLinearColor(Minimum=<rgba 0.478 0.478 0.478 0.75>, Maximum=<rgba 0.781 0.781 0.781 0.75>)
      ))
     False Vector4:Vector4f=&MakeVector4_FromLinearColor(LinearColor=&FNTimeOfDayColor(
       Night Color=&RandomRangeLinearColor(Minimum=<rgba 0.3815 0.3815 0.6313 0.75>, Maximum=<rgba 0.5522 0.5522 1 0.75>)
       Morning Color=&RandomRangeLinearColor(Minimum=<rgba 0.2565 0.2764 0.3783 0.8589>, Maximum=<rgba 0.8087 0.8582 0.9253 0.75>)
       Day Color=&RandomRangeLinearColor(Minimum=<rgba 0.5261 0.4868 0.5652 0.5>, Maximum=<rgba 0.781 0.7233 0.7233 0.7>)
       Evening Color=&RandomRangeLinearColor(Minimum=<rgba 0.4261 0.4478 0.4136 0.75>, Maximum=<rgba 0.6783 0.6478 0.6282 0.75>)
      ))
    )
    ScaleRGBA=true
   ) {enabled=false, module=ScaleColor002}
   &ScaleColor(
    Scale Alpha=&MaskFloatBySpawnGroup(
     Masked Float=&FloatFromCurve(FloatCurve=DataInterface() {propertyValues {Curve {keys [{arriveTangent=6.7279, interpMode=RCIM_Cubic, time=0, value=1}, {interpMode=RCIM_Cubic, leaveTangent=-0.9746, time=1, value=0}]}}})
     Passthrough Float=&FloatFromCurve(FloatCurve=DataInterface() {propertyValues {Curve {keys [
          {arriveTangent=-2.0221, interpMode=RCIM_Cubic, leaveTangent=-2.0221, tangentMode=RCTM_User, time=0, value=1}
          {interpMode=RCIM_Cubic, leaveTangent=-0.9746, time=1, value=0}
         ]}}})
    )
    Scale RGB=&MakeVectorFromLinearColorRGB(Color=&ColorFromCurve(ColorCurve=DataInterface() {propertyValues {
        AlphaCurve {keys []}
        BlueCurve {keys [{arriveTangent=-9.5363, interpMode=RCIM_Cubic, leaveTangent=-9.5363, tangentMode=RCTM_User, time=0, value=4}, <at 1 1 RCIM_Cubic>]}
        GreenCurve {keys [{arriveTangent=-9.5363, interpMode=RCIM_Cubic, leaveTangent=-9.5363, tangentMode=RCTM_User, time=0, value=4}, <at 1 1 RCIM_Cubic>]}
        RedCurve {keys [{arriveTangent=-9.5363, interpMode=RCIM_Cubic, leaveTangent=-9.5363, tangentMode=RCTM_User, time=0, value=4}, <at 1 1 RCIM_Cubic>]}
       }}))
    ScaleA=true
    ScaleRGB=true
   ) {enabled=false}
   &ScaleColor(
    Scale Mode=::NewEnumerator1
    Scale RGBA=&MakeCustomVector4FromBool(
     Bool:NiagaraBool=&Boolean_OrOperation(
      Bool1=&IsPlatformSetActive(New Platform Set=DataInterface() {propertyValues {Platforms {qualityLevelMask=-4}}})
      Bool2=$User.bIsFrontEndPreview
     )
     True Vector4:Vector4f=&MakeVector4_FromLinearColor(LinearColor=&FNTimeOfDayColor(
       Night Color=&RandomRangeLinearColor(Minimum=<rgba 0.5737 0.5737 0.6313 0.75>, Maximum=<rgba 0.9478 0.9478 1 0.75>)
       Morning Color=&RandomRangeLinearColor(Minimum=<rgba 0.418 0.4032 0.478 0.8589>, Maximum=<rgba 0.8207 0.8651 0.9253 0.75>)
       Day Color=&RandomRangeLinearColor(Minimum=<rgba 0.5261 0.5082 0.5032 0.5>, Maximum=<rgba 0.781 0.781 0.781 0.7>)
       Evening Color=&RandomRangeLinearColor(Minimum=<rgba 0.478 0.478 0.478 0.75>, Maximum=<rgba 0.781 0.781 0.781 0.75>)
      ))
     False Vector4:Vector4f=&MakeVector4_FromLinearColor(LinearColor=&FNTimeOfDayColor(
       Night Color=&RandomRangeLinearColor(Minimum=<rgba 0.4968 0.4968 0.6313 0.75>, Maximum=<rgba 0.7783 0.7783 1 0.75>)
       Morning Color=&RandomRangeLinearColor(Minimum=<rgba 0.2565 0.2764 0.3783 0.8589>, Maximum=<rgba 0.8087 0.8582 0.9253 0.75>)
       Day Color=&RandomRangeLinearColor(Minimum=<rgba 0.5261 0.4868 0.5652 0.5>, Maximum=<rgba 0.781 0.7233 0.7233 0.7>)
       Evening Color=&RandomRangeLinearColor(Minimum=<rgba 0.4261 0.4478 0.4136 0.75>, Maximum=<rgba 0.6783 0.6478 0.6282 0.75>)
      ))
    )
    ScaleRGBA=true
   ) {module=ScaleColor001}
   &ScaleColor(
    Scale Alpha=&FloatFromCurve(
     FloatCurve=DataInterface() {propertyValues {Curve {keys [
         {arriveTangent=2.6927, interpMode=RCIM_Cubic, leaveTangent=2.6927, tangentMode=RCTM_User, time=0, value=0}
         {arriveTangent=0.1984, leaveTangent=0.0132, tangentMode=RCTM_User, time=0.6, value=0.7}
         <at 2 0.6>
         {arriveTangent=-0.1351, interpMode=RCIM_Cubic, leaveTangent=-0.3618, tangentMode=RCTM_Break, time=3.2, value=0.4}
         {arriveTangent=-0.225, tangentMode=RCTM_User, time=5, value=0}
        ]}}}
     CurveIndex=$Emitter.Age
    )
    ScaleA=true
    ScaleRGB=false
   ) {module=ScaleColor003}
   &DynamicMaterialParameters(
    Index 0 Param 1=&Add_Float(
     A=&RandomRangeFloat(Minimum=0, Maximum=1)
     B=&FloatFromCurve(
      FloatCurve=DataInterface() {propertyValues {Curve {keys [<at 0 0 RCIM_Cubic>, {arriveTangent=0.1107, leaveTangent=0.1107, time=2.5, value=0.25}, <at 5 1.05 RCIM_Cubic>]}}}
      CurveIndex=&Add_Float(A=&Multiply_Float(A=$Particles.Initial.MaterialRandom, B=1), B=$Emitter.Age)
      Scale Curve=0.1
     )
    )
    Index 0 Param 2=&FloatFromCurve(FloatCurve=DataInterface() {propertyValues {Curve {keys [<at 0 1>,<at 1 0>]}}})
    Index 0 Param 3=&FloatFromCurve(
     FloatCurve=DataInterface() {propertyValues {Curve {keys [
         {interpMode=RCIM_Cubic, leaveTangent=16.5652, leaveTangentWeight=0.4385, tangentMode=RCTM_Break, time=0, value=0}
         {arriveTangentWeight=0.0333, interpMode=RCIM_Cubic, leaveTangentWeight=0.9477, tangentMode=RCTM_Break, time=0.1, value=1}
         {arriveTangent=-1.7159, interpMode=RCIM_Cubic, leaveTangent=-1.7159, tangentMode=RCTM_User, time=0.4, value=0.8}
         {arriveTangent=0.2347, interpMode=RCIM_Cubic, leaveTangent=0.2723, tangentMode=RCTM_Break, time=0.6, value=0}
        ]}}}
     Scale Curve=0
    )
    Index 0 Param 4=&FloatFromCurve(FloatCurve=DataInterface() {propertyValues {Curve {keys [{interpMode=RCIM_Cubic, tangentMode=RCTM_Break, time=0.5, value=0.2}, {interpMode=RCIM_Cubic, tangentMode=RCTM_Break, time=1, value=0}]}}})
    Index 1 Param 1=&FloatFromCurve(FloatCurve=DataInterface() {propertyValues {Curve {keys [
         {interpMode=RCIM_Cubic, leaveTangentWeight=0.0833, tangentMode=RCTM_Break, time=0, value=0}
         {arriveTangentWeight=0.0873, interpMode=RCIM_Cubic, tangentMode=RCTM_Break, time=0.25, value=1}
        ]}}})
    Index 1 Param 2=&FloatFromCurve(
     FloatCurve=DataInterface() {propertyValues {Curve {keys [
         {interpMode=RCIM_Cubic, tangentMode=RCTM_Break, time=0, value=0.2}
         {interpMode=RCIM_Cubic, leaveTangent=0.9066, tangentMode=RCTM_Break, time=1, value=2}
        ]}}}
     Scale Curve=&RandomRangeFloat(Minimum=0.1, Maximum=0.2)
    )
    Index 1 Param 3=&FloatFromCurve(FloatCurve=DataInterface() {propertyValues {Curve {keys [
         {interpMode=RCIM_Cubic, leaveTangent=6.3917, tangentMode=RCTM_Break, time=0, value=0}
         {arriveTangent=1.7549, interpMode=RCIM_Cubic, leaveTangent=10.6767, tangentMode=RCTM_User, time=0.25, value=1}
        ]}}})
    Index 1 Param 4=&FloatFromCurve(
     FloatCurve=DataInterface() {propertyValues {Curve {keys [{arriveTangent=-0.136, interpMode=RCIM_Cubic, leaveTangent=-0.136, tangentMode=RCTM_User, time=0, value=1}, <at 1 0.95>]}}}
     Scale Curve=2
    )
    Index 2 Param 1=&FloatFromCurve(
     FloatCurve=DataInterface() {propertyValues {Curve {keys [
         {interpMode=RCIM_Cubic, leaveTangent=0.4766, tangentMode=RCTM_Break, time=0, value=0}
         {arriveTangent=1.0548, interpMode=RCIM_Cubic, leaveTangent=1.0548, tangentMode=RCTM_User, time=0.5, value=0.45}
         <at 1 1>
        ]}}}
     Scale Curve=3
    )
    Index 2 Param 2=&FloatFromCurve(
     FloatCurve=DataInterface() {propertyValues {Curve {keys [{interpMode=RCIM_Cubic, tangentMode=RCTM_User, time=0, value=0}, {interpMode=RCIM_Cubic, tangentMode=RCTM_User, time=1, value=1}]}}}
     Scale Curve=0.3
    )
    Index 2 Param 3=&FloatFromCurve(
     FloatCurve=DataInterface() {propertyValues {Curve {keys [
         {interpMode=RCIM_Cubic, leaveTangent=-15.7052, leaveTangentWeight=0.1411, tangentMode=RCTM_Break, time=0, value=1}
         {interpMode=RCIM_Cubic, tangentMode=RCTM_Break, time=0.1, value=0.2}
         {arriveTangent=0.7532, interpMode=RCIM_Cubic, leaveTangent=0.7532, tangentMode=RCTM_User, time=0.5, value=0.45}
         {arriveTangent=0.4925, arriveTangentWeight=1.057, interpMode=RCIM_Cubic, tangentMode=RCTM_Break, time=1, value=0.75}
        ]}}}
     Scale Curve=5
    )
    Index 2 Param 4=&RandomRangeFloat(Minimum=1.5, Maximum=1.75)
    Module.Param0WriteEnabled=true
    Module.Param0WriteEnabled1=true
    Module.Param0WriteEnabled2=true
    Module.Param1WriteEnabled=false
    Module.Param1WriteEnabled1=true
    Module.Param1WriteEnabled2=true
    Module.Param2WriteEnabled=true
    Module.Param2WriteEnabled1=true
    Module.Param2WriteEnabled2=true
    Module.Param3WriteEnabled=true
    Module.Param3WriteEnabled1=true
    Module.Param3WriteEnabled2=true
    Write Parameter Index 1=true
    Write Parameter Index 2=true
   ) {module=DynamicMaterialParameters001}
  ]
  rendererClasses [&NiagaraSpriteRendererProperties]
  renderers [{rendererClass=&NiagaraSpriteRendererProperties}]
 }
 Sparks_CPU {
  emitterUpdateScript [
   &EmitterState(
    Enable Distance Culling=true
    Life Cycle Mode=::NewEnumerator1
    Loop Behavior=::NewEnumerator1
    Loop Delay=0.26
    Loop Duration=2
    Max Distance Response=::KillImmediately
    Module.bApplySpawnCountScaleByDistance=true
    Module.bUseMaxDistance=true
    Module.bUseMinDistance=false
    Scalability Mode=::NewEnumerator1
    Scale Spawn Count=true
    Spawn Count Scale By Distance=DataInterface() {propertyValues {Curve {keys [<at 0.3676 0.9657>,<at 0.9962 0.4982>]}}}
    UseLoopDelay=true
   )
   &SpawnBurst_Instantaneous(Spawn Count=&UniformRangedInt(Minimum=16, Maximum=24), Use Spawn Probability=false)
  ]
  particleSpawnScript [
   &InitializeParticle(
    Color=&RandomRangeLinearColor(Minimum=<rgba 5 4.4167 1.5 1>, Maximum=<rgba 10 10 2 1>)
    Lifetime Max=1.8
    Lifetime Min=1.328
    Lifetime Mode=::NewEnumerator1
    Sprite Size Max=<1.25 4>
    Sprite Size Min=<1 3>
    Sprite Size Mode=::NewEnumerator2
    UsePositionOffset=false
    Write Material Random=true
   )
   &SkeletalMeshLocation(
    Bone / Socket Index=3
    MaskSpawnGroup=false
    Mesh Sampling Type=::NewEnumerator5
    Orientation Sampling=::NewEnumerator1
    Position Sampling=::NewEnumerator2
    Sampled Position Offset=<10 15 -5>
    Sampled Position Offset Coordinate Space:ENiagaraSkelMeshTransforms=::NewEnumerator2
    Skeletal Mesh=$System.Skeletal Mesh
    Socket Sampling Mode=::NewEnumerator5
    Use Sampled Mesh Attribute as Ribbon ID=false
   ) {module=SkeletalMeshLocation001}
   &ShapeLocation(
    MaskSpawnGroup=false
    Override Seed=false
    Shape Origin=$Particles.SkeletalMeshLocation001.SampledPosition
    Sphere Radius=9
    Sphere Surface Distribution=1
    UseOwnerScale=true
   )
   &AddVelocity(
    Module.Use Falloff Distance=false
    Override Seed=false
    Velocity Mode=::NewEnumerator1
    Velocity Origin=$Particles.SkeletalMeshLocation001.SampledPosition
    Velocity Speed=&FN_FrontEndFloatModifier(
     Gameplay=&RandomRangeFloat(Minimum=111, Maximum=333)
     Front End=&RandomRangeFloat(Minimum=75, Maximum=200)
    )
   )
   &DynamicMaterialParameters(
    Index 0 Param 1=3
    Index 0 Param 2=20
    Index 0 Param 3=1
    Module.Param0WriteEnabled=true
    Module.Param1WriteEnabled=true
    Module.Param2WriteEnabled=true
    Module.Param3WriteEnabled=true
   )
  ]
  particleUpdateScript [
   &ParticleState()
   &GravityForce(Gravity=<0 0 -10>)
   &Drag(Drag=&UniformRangedFloat(Minimum=2, Maximum=3), Use Linear Drag=true, Use Rotational Drag=false)
   &SolveForcesAndVelocity(Clamp Velocity=false, Limit Acceleration=false)
   &ScaleSpriteSize(Scale Factor=&Vector2DFromCurve(
     Vector2Curve=DataInterface() {propertyValues {
       XCurve {keys [
         {arriveTangent=5.9728, interpMode=RCIM_Cubic, leaveTangent=5.9728, tangentMode=RCTM_User, time=0, value=0.1}
         {arriveTangent=-0.1, interpMode=RCIM_Cubic, leaveTangent=-0.1, tangentMode=RCTM_User, time=0.18, value=0.8}
         {interpMode=RCIM_Cubic, tangentMode=RCTM_User, time=1, value=0}
        ]}
       YCurve {keys [
         {arriveTangent=6.2715, interpMode=RCIM_Cubic, leaveTangent=6.2715, tangentMode=RCTM_User, time=0, value=0.1}
         {arriveTangent=-0.1001, interpMode=RCIM_Cubic, leaveTangent=-0.1001, tangentMode=RCTM_User, time=0.18, value=0.8}
         {interpMode=RCIM_Cubic, tangentMode=RCTM_User, time=0.9988, value=0}
        ]}
      }}
     ScaleCurve=<1 1>
    ))
   &ScaleSpriteSizeBySpeed(Max Scale Factor=<1 1.5>, Min Scale Factor=<0.5 1>, Velocity Threshold=666)
   &ScaleColor(
    Scale RGB=&VectorFromFloat(Value=&FloatFromCurve(
      FloatCurve=DataInterface() {propertyValues {Curve {keys [
          {interpMode=RCIM_Cubic, tangentMode=RCTM_User, time=0, value=1}
          <at 0.3 0.2>
          {interpMode=RCIM_Cubic, tangentMode=RCTM_User, time=1, value=0}
         ]}}}
      Scale Curve=100
     ))
    ScaleA=true
    ScaleRGB=true
   )
   /SetVariables(
    Particles.GlowColor=&MakeLinearColorFromVectorAndFloat(
     "Vector (RGB)"=&MakeVectorFromLinearColorRGB(Color=<rgba 1 0.9417 0.3 1>)
     "Float (Alpha)"=&FloatFromCurve(FloatCurve=DataInterface() {propertyValues {Curve {keys [<at 0.5 1>,<at 1 0>]}}}, Scale Curve=0.2)
    )
    Particles.GlowSize=&Multiply_Vector2D(A=$Particles.SpriteSize, B=<8 8>)
   ) {isSetParameters=true}
  ]
  rendererClasses [&NiagaraSpriteRendererProperties]
  renderers [
   {rendererClass=&NiagaraSpriteRendererProperties}
   {rendererClass=&NiagaraSpriteRendererProperties, rendererIndex=1}
  ]
 }
}
summary {
 systemName=NS_Emote_WealthLamb_FateStork_Meteor
 emitters [
  {emitterName=Main_Meteor, rendererClasses [&NiagaraMeshRendererProperties]}
  {
   emitterName=HT_Debris
   simTarget=GPUComputeSim
   rendererClasses [&NiagaraMeshRendererProperties, &NiagaraSpriteRendererProperties]
  }
  {emitterName=ImpactDust, rendererClasses [&NiagaraSpriteRendererProperties]}
  {emitterName=SmokeBIG, rendererClasses [&NiagaraSpriteRendererProperties]}
  {emitterName=SmokeBIGLOWMID, rendererClasses [&NiagaraSpriteRendererProperties]}
  {
   emitterName=Flare_Complex001
   rendererClasses [
    &NiagaraSpriteRendererProperties
    &NiagaraLightRendererProperties
    &NiagaraMeshRendererProperties
   ]
  }
  {emitterName=Sparks_CPU, rendererClasses [&NiagaraSpriteRendererProperties]}
  {emitterName=FlashElements001, rendererClasses [&NiagaraSpriteRendererProperties]}
 ]
 userVariables [
  User.bisFootHidden {type=NiagaraBool}
  User.bIsFrontEnd {type=NiagaraBool}
  User.bIsFrontEndPreview {type=NiagaraBool}
  User.Debug {type=NiagaraBool}
  User.DebugVelocity {type=Vector3f, description="", defaultValue {struct=Vector3f, value=<0 0 1>}}
  User.DesertBoulder {type=NiagaraBool}
  User.EmissiveColor {type=LinearColor, description="", defaultValue {struct=LinearColor, value=<rgba 0.0196 0.2745 0.5255 1>}}
  User.FrontEndScale {type=NiagaraFloat, defaultValue=0.6, description=""}
  User.ParaGlideLeanAlpha {type=NiagaraFloat, defaultValue=0, description=""}
  User.RibbonIDOffset {type=NiagaraInt32}
  User.SkeletalMesh {
   type=NiagaraDataInterfaceSkeletalMesh
   defaultValue {
    struct=NiagaraExt_VariableValue_DataInterface
    value {dataInterface=&self:NiagaraDataInterfaceSkeletalMesh_0}
   }
  }
 ]
}
