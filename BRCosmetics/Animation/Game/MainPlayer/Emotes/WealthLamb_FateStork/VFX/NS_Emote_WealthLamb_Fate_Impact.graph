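# Niagara impact VFX for the WealthLamb FateStork emote: a brief, blue-tinted burst of flares, sparks and debris.
# Five burst emitters supply a radial flare with spikes, star flash with light, glowing sparks, falling crystal mesh fragments and triangle meshes.
# Custom per-particle attributes let one particle drive layered sprite renderers with separate colour, size or age.
# Front-end modifiers adjust scale, System.TOD tints several layers, and sparks cull by distance.
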NiagaraSystem /BRCosmetics/Animation/Game/MainPlayer/Emotes/WealthLamb_FateStork/VFX/NS_Emote_WealthLamb_Fate_Impact
dependencies {
 usedDataInterfaces [
  &NiagaraDataInterfaceCurve
  &NiagaraDataInterfaceMeshRendererInfo
  &NiagaraDataInterfaceVector2DCurve
  &NiagaraDataInterfaceColorCurve
  &NiagaraDataInterfaceCamera
 ]
 usedDynamicInputs [
  &MergeHelper_Vector2Switch
  &RandomRangeVector2D
  &ReturnCameraVector
  &ReturnNormalizedExecIndex
 ]
 usedModules [
  &self:SystemSpawnScript.SystemScriptSource.SystemScriptGraph.NiagaraNodeAssignment_0.SetVariables_#1
  &SystemState
 ]
}
emitters {
 ChromaRadialFlare {
  emitterUpdateScript [
   &EmitterState(Life Cycle Mode=::NewEnumerator1, Loop Behavior=::NewEnumerator1, Loop Duration=1, UseLoopDelay=false)
   &SpawnBurst_Instantaneous(Use Spawn Probability=false, UseLoopCountLimit=false) {module=SpawnBurst_Instantaneous002}
  ]
  particleSpawnScript [
   &InitializeParticle(
    Color=&MultiplyLinearColor(Linear Color=<rgba 1 1 1 1>, Scale Factor=&MakeVector4_FromLinearColor(LinearColor=$System.TOD))
    Lifetime=0.275
    Sprite Rotation Mode=::NewEnumerator2
    Sprite Size Mode=::NewEnumerator3
    Uniform Sprite Size=&Multiply_Float(A=200, B=&FN_FrontEndFloatModifier(Gameplay=1, Front End=$User.FrontEndScale))
    UsePositionOffset=false
    Write Material Random=true
   ) {module=InitializeParticle001}
   &CameraOffset(Camera Offset Amount=50) {module=CameraOffset001}
  ]
  particleUpdateScript [
   &ParticleState()
   &ScaleColor(
    Scale Alpha=&FloatFromCurve(
     FloatCurve=DataInterface() {propertyValues {Curve {keys [{arriveTangent=-3.3955, interpMode=RCIM_Cubic, leaveTangent=-3.3955, tangentMode=RCTM_User, time=0.2, value=1}, <at 1 0 RCIM_Cubic>]}}}
     Scale Curve=0.5
    )
    Scale RGB=&MakeVectorFromLinearColorRGB(Color=&ColorFromCurve(ColorCurve=DataInterface() {propertyValues {
        AlphaCurve {keys [<at 0 1 RCIM_Cubic>]}
        BlueCurve {keys [<at 0 1 RCIM_Cubic>,<at 0.5354 1>,<at 0.9851 0.5227>]}
        GreenCurve {keys [<at 0 0.8848 RCIM_Cubic>,<at 0.5354 0.4697>,<at 0.9851 0.8085>]}
        RedCurve {keys [<at 0 0.306 RCIM_Cubic>,<at 0.5354 0.6193>,<at 0.9851 1>]}
       }}))
    ScaleA=true
    ScaleRGB=true
   )
   &ScaleSpriteSize(Scale Factor=&Vector2DFromFloat(Value=&FloatFromCurve(FloatCurve=DataInterface() {propertyValues {Curve {keys [{arriveTangent=1.7733, interpMode=RCIM_Cubic, leaveTangent=1.7733, tangentMode=RCTM_User, time=0, value=0.2}, <at 1 1 RCIM_Cubic>]}}})))
   &DynamicMaterialParameters(
    Index 0 Param 1=&FloatFromCurve(FloatCurve=DataInterface() {propertyValues {Curve {keys [{tangentMode=RCTM_User, time=0, value=0}, {tangentMode=RCTM_User, time=1, value=1}]}}})
    Index 0 Param 2=&FloatFromCurve(FloatCurve=DataInterface() {propertyValues {Curve {keys [
         {arriveTangent=2.4836, interpMode=RCIM_Cubic, leaveTangent=2.4836, tangentMode=RCTM_User, time=0, value=0}
         {interpMode=RCIM_Cubic, tangentMode=RCTM_User, time=1, value=1}
        ]}}})
    Index 1 Param 1=3
    Index 1 Param 2=10
    Index 1 Param 3=1
    Index 1 Param 4=1
    Module.Index0Param0Write=true
    Module.Index0Param1Write=true
    Module.Index0Param2Write=true
    Module.Index0Param3Write=true
    Module.Index1Param0Write=true
    Module.Index1Param1Write=true
    Module.Index1Param2Write=true
    Module.Index1Param3Write=true
    Write Parameter Index 1=true
   ) {module=DynamicMaterialParameters001}
   /SetVariables(
    Particles.SpikesColor:LinearColor=&MakeLinearColorFromVectorAndFloat(
     "Vector (RGB)"=&MakeVectorFromLinearColorRGB(Color=&ColorFromCurve(ColorCurve=DataInterface() {propertyValues {
         AlphaCurve {keys []}
         BlueCurve {keys [<at 0.0027 3>,<at 0.931 0.2 RCIM_Cubic>]}
         GreenCurve {keys [<at 0.0027 1.2>,<at 0.931 0.05 RCIM_Cubic>]}
         RedCurve {keys [<at 0.0027 0.3>,<at 0.931 0.02 RCIM_Cubic>]}
        }}))
     "Float (Alpha)"=&FloatFromCurve(
      FloatCurve=DataInterface() {propertyValues {Curve {keys [{interpMode=RCIM_Cubic, tangentMode=RCTM_User, time=0, value=1}, {interpMode=RCIM_Cubic, tangentMode=RCTM_User, time=1, value=0}]}}}
      Scale Curve=2.5
     )
    )
    Particles.SpikesSpriteSize:Vector2f=&Multiply_Vector2D_ByFloat(
     Float=&FloatFromCurve(
      FloatCurve=DataInterface() {propertyValues {Curve {keys [{tangentMode=RCTM_User, time=0, value=0.8}, {tangentMode=RCTM_User, time=1, value=1}]}}}
      Scale Curve=&Multiply_Float(
       A=&RandomRangeFloat(Minimum=0.8, Maximum=1)
       B=&FN_FrontEndFloatModifier(Gameplay=0.8, Front End=0.4)
      )
     )
     Vector2D=<200 175>
    )
   ) {isSetParameters=true}
  ]
  rendererClasses [&NiagaraSpriteRendererProperties]
  renderers [
   {rendererClass=&NiagaraSpriteRendererProperties}
   {rendererClass=&NiagaraSpriteRendererProperties, rendererIndex=1}
  ]
 }
 CrystalFragments {
  emitterUpdateScript [
   &EmitterState()
   &SpawnBurst_Instantaneous(
    Spawn Count=&RandomRangeInt(Minimum=4, Maximum=6)
    Use Spawn Probability=false
    UseLoopCountLimit=false
   )
  ]
  particleSpawnScript [
   /SetVariables(Particles.BiasedRandom=&FloatFromCurve(
     FloatCurve=DataInterface() {propertyValues {Curve {keys [
         <at 0.2 1 RCIM_Cubic>
         {arriveTangent=-3.3394, interpMode=RCIM_Cubic, leaveTangent=-3.3394, tangentMode=RCTM_User, time=0.5, value=0.5}
         <at 0.8 0 RCIM_Cubic>
        ]}}}
     CurveIndex=&RandomRangeFloat(Minimum=0, Maximum=1)
    )) {isSetParameters=true}
   &InitializeParticle(
    Color=<rgba 10 10 10 1>
    Lifetime=&RandomRangeFloat(Minimum=0.5, Maximum=1)
    Mass=&Lerp_Float(A=0.4, B=0.8, Alpha=$Particles.BiasedRandom)
    Mass Mode=::NewEnumerator1
    Mesh Scale=&Multiply_VectorByFloat(
     Vector=&RandomRangeVector(Minimum=<0.3 0.3 2>, Maximum=<0.6 0.6 3>)
     Float=&Lerp_Float(A=0.5, B=1, Alpha=$Particles.BiasedRandom)
    )
    Mesh Scale Mode=::NewEnumerator4
    UsePositionOffset=false
    Write Material Random=true
   )
   &SphereLocation(
    MaskSpawnGroup=false
    Override Seed=false
    Sphere Radius=1
    Surface Only=false
    UseNonUniformScale=false
    UseOffset=true
    UseSphereOrientation=false
   )
   &AddVelocityInCone(
    Cone Angle=150
    Cone Axis=<1 0 1>
    Module.Use Velocity Falloff On Cone Axis=false
    Velocity Strength=&RandomRangeFloat(Minimum=80, Maximum=300)
   )
   &InitialMeshOrientation(Use Rotation Vector=false)
   &AddRotationalVelocity(Rotation Rate=&RandomVector(Vector Scale=&RandomRangeFloat(Minimum=2, Maximum=3)))
  ]
  particleUpdateScript [
   &ParticleState()
   &GravityForce(Gravity=&RandomRangeVector(Minimum=<0 0 -800>, Maximum=<0 0 -1200>))
   &SolveForcesAndVelocity(Clamp Velocity=false, Limit Acceleration=false)
   &ScaleMeshSize(Scale Factor=&VectorFromFloat(Value=&FloatFromCurve(
      FloatCurve=DataInterface() {propertyValues {Curve {keys [<at 0.6 1 RCIM_Cubic>, {arriveTangent=-6.308, interpMode=RCIM_Cubic, leaveTangent=-6.308, tangentMode=RCTM_User, time=1, value=0}]}}}
      Scale Curve=&FN_FrontEndFloatModifier(Gameplay=1, Front End=$System.FrontendScale)
     ))) {module=ScaleMeshSize001}
   &ScaleColor(
    Scale Alpha=&FloatFromCurve(FloatCurve=DataInterface() {propertyValues {Curve {keys [<at 0.7 1 RCIM_Cubic>,<at 1 1>]}}})
    Scale RGB=&MakeVectorFromLinearColorRGB(Color=&ColorFromCurve(ColorCurve=DataInterface() {propertyValues {
        AlphaCurve {keys []}
        BlueCurve {keys [<at 0 0.6615 RCIM_Cubic>,<at 0.6835 1>,<at 1 0 RCIM_Cubic>]}
        GreenCurve {keys [<at 0 0.6573 RCIM_Cubic>,<at 0.6835 1>,<at 1 0 RCIM_Cubic>]}
        RedCurve {keys [<at 0 0.6442 RCIM_Cubic>,<at 0.6835 1>,<at 1 0 RCIM_Cubic>]}
       }}))
    ScaleA=true
    ScaleRGB=true
   )
  ]
  rendererClasses [&NiagaraMeshRendererProperties]
  renderers [{rendererClass=&NiagaraMeshRendererProperties, rendererIndex=0}]
 }
 FlashElements001 {
  emitterUpdateScript [
   &EmitterState()
   &SpawnBurst_Instantaneous(Spawn Group=1, Use Spawn Probability=false, UseLoopCountLimit=false) {module=SpawnBurst_Instantaneous001}
   &SpawnBurst_Instantaneous(
    Spawn Count=&RandomRangeInt(Minimum=6, Maximum=8)
    Use Spawn Probability=false
    UseLoopCountLimit=false
   )
  ]
  particleSpawnScript [
   &InitializeParticle(Color Mode=::NewEnumerator0, Lifetime=0.25, Sprite Rotation Mode=::NewEnumerator2, UsePositionOffset=false, Write Material Random=true)
   &ShapeLocation(MaskSpawnGroup=false, Override Seed=false, Sphere Radius=1, UseOwnerScale=true)
   &SpriteFacingAndAlignment(Sprite Alignment=$Particles.ShapeLocation.ShapeNormal, WriteAlignment=true, WriteFacing=false)
   &FN_PerParticleLOD(Enable SpriteSize Distance Scaling=true, Sprite Size Scale Far=<2.5 2.5>)
   /SetVariables(Particles.SpriteSize=<1 1>, Particles.VisibilityTag=&SpawnGroup()) {isSetParameters=true}
   /SetVariables(
    Particles.Flash Star Cam Offset=15
    Particles.Flash Star Color:LinearColor=$System.TOD Color
    Particles.Flash Star Lifetime=0.15
    Particles.Flash Star Rotation:NiagaraFloat=&RandomRangeFloat(Minimum=0, Maximum=360)
    Particles.Flash Star Sprite Size:Vector2f=&Multiply_Vector2D_ByFloat(
     Float=&Multiply_Float(
      A=&MakeCustomFloatFromBool(Bool=$User.bIsFrontEnd, True Float=$User.FrontEndScale, False Float=1)
      B=&RandomRangeFloat(Minimum=110, Maximum=130)
     )
     Vector2D=$Particles.SpriteSize
    )
    Particles.Flash Star SubUV Index:NiagaraFloat=&RandomRangeFloat(Minimum=0, Maximum=3)
   ) {isSetParameters=true}
   /SetVariables(
    Particles.Flash Flare Color:LinearColor=$System.TOD Color
    Particles.Flash Flare Lifetime=0.15
    Particles.Flash Flare Sprite Size:Vector2f=&Multiply_Vector2D_ByFloat(Float=&FN_FrontEndFloatModifier(Gameplay=1, Front End=$User.FrontEndScale), Vector2D=<180 180>)
   ) {isSetParameters=true}
   &DynamicMaterialParameters(
    Index 0 Param 1=0.2
    Index 0 Param 2=1
    Module.Index0Param0Write=true
    Module.Index0Param1Write=true
    Module.Index0Param2Write=true
    Module.Index0Param3Write=true
   )
   &CameraOffset(Camera Offset Amount=15)
  ]
  particleUpdateScript [
   &ParticleState()
   /SetVariables(
    Particles.Flash Flare Age:NiagaraFloat=&Divide_Float(A=$Particles.Age, B=$Particles.Initial.Flash Flare Lifetime)
    Particles.Flash Star Age:NiagaraFloat=&Divide_Float(A=$Particles.Age, B=$Particles.Initial.Flash Star Lifetime)
   ) {isSetParameters=true}
   /SetVariables(
    Particles.Flash Flare Color:LinearColor=&MakeLinearColorFromVectorAndFloat(
     "Vector (RGB)"=&Multiply_Vector(
      A=&Multiply_Vector(
       A=&MakeVectorFromLinearColorRGB(Color=$Particles.Initial.Flash Flare Color)
       B=&MakeVectorFromLinearColorRGB(Color=&ColorFromCurve(
         ColorCurve=DataInterface() {propertyValues {
           AlphaCurve {keys [<at 0 1 RCIM_Cubic>,<at 1 0 RCIM_Cubic>]}
           BlueCurve {keys [<at 0.0198 2>,<at 1 2>]}
           GreenCurve {keys [<at 0.0198 0.8667>,<at 1 0.6667>]}
           RedCurve {keys [<at 0.0198 0.3>,<at 1 0.4>]}
          }}
         CurveIndex=$Particles.Flash Flare Age
        ))
      )
      B=&VectorFromFloat(Value=&FN_FrontEndFloatModifier(Gameplay=1, Front End=0.75))
     )
     "Float (Alpha)"=&Multiply_Float(
      A=&MakeFloatFromLinearColor(LinearColor=<rgba 1 1 1 1>, Channel:ENiagaraLinearColor_Channels=::NewEnumerator3)
      B=&FloatFromCurve(
       FloatCurve=DataInterface() {propertyValues {Curve {keys [{arriveTangent=-1.9727, interpMode=RCIM_Cubic, time=0.1, value=1}, <at 1 0 RCIM_Cubic>]}}}
       CurveIndex=$Particles.Flash Flare Age
      )
     )
    )
    Particles.Flash Star Color:LinearColor=&MultiplyLinearColor(
     Linear Color=$Particles.Initial.Flash Star Color
     Scale Factor=&MakeVector4_FromLinearColor(LinearColor=&ColorFromCurve(
       ColorCurve=DataInterface() {propertyValues {
         AlphaCurve {keys [<at 0.2935 1>,<at 0.9862 0>]}
         BlueCurve {keys [<at 0 3>,<at 1 1>]}
         GreenCurve {keys [<at 0 1.3>,<at 1 0.2917>]}
         RedCurve {keys [<at 0 0.45>,<at 1 0.15>]}
        }}
       CurveIndex=$Particles.Flash Star Age
      ))
    )
    Particles.Flash Star Sprite Size:Vector2f=&Multiply_Vector2D_ByFloat(
     Float=&FloatFromCurve(
      FloatCurve=DataInterface() {propertyValues {Curve {keys [
          {arriveTangent=2.7013, interpMode=RCIM_Cubic, leaveTangent=2.7013, tangentMode=RCTM_User, time=0, value=0}
          {interpMode=RCIM_Cubic, tangentMode=RCTM_User, time=1, value=1}
         ]}}}
      CurveIndex=$Particles.Flash Star Age
     )
     Vector2D=$Particles.Initial.Flash Star Sprite Size
    )
   ) {isSetParameters=true}
   &FN_Light_Attributes(
    Color=&MultiplyLinearColor(
     Linear Color=<rgba 0.15 0.3625 1 1>
     Scale Factor=&MakeVector4_FromLinearColor(LinearColor=&LerpLinearColors(StartColor=<rgba 1 1 1 1>, EndColor=$System.TOD Color, LerpFactor=2))
    )
    Create Color Attribute=true
    Create Exponent Attribute=true
    Create Position Attribute=true
    Create Radius Attribute=true
    Exponent=1
    Radius=&FN_FrontEndFloatModifier(Gameplay=90, Front End=90)
    Scale Color=&FloatFromCurve(
     FloatCurve=DataInterface() {propertyValues {Curve {keys [
         {arriveTangent=-0.5504, interpMode=RCIM_Cubic, leaveTangent=-0.5504, tangentMode=RCTM_User, time=0, value=1}
         {arriveTangent=-0.5426, interpMode=RCIM_Cubic, leaveTangent=-0.5426, tangentMode=RCTM_User, time=0.55, value=0}
        ]}}}
     Scale Curve=2
    )
   )
  ]
  rendererClasses [&NiagaraSpriteRendererProperties, &NiagaraLightRendererProperties]
  renderers [
   {rendererClass=&NiagaraSpriteRendererProperties}
   {rendererClass=&NiagaraSpriteRendererProperties, rendererIndex=1}
   {rendererClass=&NiagaraLightRendererProperties, rendererIndex=2}
  ]
 }
 Sparks_CPU {
  emitterUpdateScript [
   &EmitterState(
    Enable Distance Culling=true
    Max Distance Response=::KillImmediately
    Module.bApplySpawnCountScaleByDistance=true
    Module.bUseMaxDistance=true
    Module.bUseMinDistance=false
    Scalability Mode=::NewEnumerator1
    Scale Spawn Count=true
    Spawn Count Scale By Distance=DataInterface() {propertyValues {Curve {keys [<at 0.3676 0.9657>,<at 0.9962 0.4982>]}}}
   )
   &SpawnBurst_Instantaneous(Spawn Count=&UniformRangedInt(Minimum=16, Maximum=24), Use Spawn Probability=false)
  ]
  particleSpawnScript [
   &InitializeParticle(
    Color=&RandomRangeLinearColor(Minimum=<rgba 1.5 3.8333 5 1>, Maximum=<rgba 2 6 10 1>)
    Lifetime Max=0.45
    Lifetime Min=0.166
    Lifetime Mode=::NewEnumerator1
    Sprite Size Mode=::NewEnumerator3
    Uniform Sprite Size=&Multiply_Float(
     A=&MakeCustomFloatFromBool(Bool=$User.bIsFrontEnd, True Float=0.45, False Float=1)
     B=&RandomRangeFloat(Minimum=2.5, Maximum=7)
    )
    UsePositionOffset=false
    Write Material Random=true
   )
   &ShapeLocation(MaskSpawnGroup=false, Override Seed=false, Sphere Radius=9, Sphere Surface Distribution=1, UseOwnerScale=true)
   &AddVelocity(
    Module.Use Falloff Distance=false
    Override Seed=false
    Velocity Mode=::NewEnumerator1
    Velocity Speed=&FN_FrontEndFloatModifier(
     Gameplay=&RandomRangeFloat(Minimum=222, Maximum=666)
     Front End=&RandomRangeFloat(Minimum=150, Maximum=400)
    )
   )
   &DynamicMaterialParameters(
    Index 0 Param 1=3
    Index 0 Param 2=20
    Index 0 Param 3=1
    Module.Param0WriteEnabled=true
    Module.Param1WriteEnabled=true
    Module.Param2WriteEnabled=true
    Module.Param3WriteEnabled=true
   )
  ]
  particleUpdateScript [
   &ParticleState()
   &GravityForce(Gravity=<0 0 -666>)
   &Drag(Drag=&UniformRangedFloat(Minimum=2, Maximum=6), Use Linear Drag=true, Use Rotational Drag=false)
   &SolveForcesAndVelocity(Clamp Velocity=false, Limit Acceleration=false)
   &ScaleSpriteSize(Scale Factor=&Vector2DFromCurve(
     Vector2Curve=DataInterface() {propertyValues {
       XCurve {keys [
         {arriveTangent=5.9728, interpMode=RCIM_Cubic, leaveTangent=5.9728, tangentMode=RCTM_User, time=0, value=0.1}
         {arriveTangent=-0.1, interpMode=RCIM_Cubic, leaveTangent=-0.1, tangentMode=RCTM_User, time=0.18, value=0.8}
         {interpMode=RCIM_Cubic, tangentMode=RCTM_User, time=1, value=0}
        ]}
       YCurve {keys [
         {arriveTangent=6.2715, interpMode=RCIM_Cubic, leaveTangent=6.2715, tangentMode=RCTM_User, time=0, value=0.1}
         {arriveTangent=-0.1001, interpMode=RCIM_Cubic, leaveTangent=-0.1001, tangentMode=RCTM_User, time=0.18, value=0.8}
         {interpMode=RCIM_Cubic, tangentMode=RCTM_User, time=0.9988, value=0}
        ]}
      }}
     ScaleCurve=<1 1>
    ))
   &ScaleSpriteSizeBySpeed(Max Scale Factor=<0.5 3>, Min Scale Factor=<0.5 1>, Velocity Threshold=666)
   &ScaleColor(
    Scale RGB=&VectorFromFloat(Value=&FloatFromCurve(
      FloatCurve=DataInterface() {propertyValues {Curve {keys [
          {interpMode=RCIM_Cubic, tangentMode=RCTM_User, time=0, value=1}
          <at 0.3 0.2>
          {interpMode=RCIM_Cubic, tangentMode=RCTM_User, time=1, value=0}
         ]}}}
      Scale Curve=100
     ))
    ScaleA=true
    ScaleRGB=true
   )
   /SetVariables(
    Particles.GlowColor=&MakeLinearColorFromVectorAndFloat(
     "Vector (RGB)"=&MakeVectorFromLinearColorRGB(Color=<rgba 0.3 0.475 1 1>)
     "Float (Alpha)"=&FloatFromCurve(FloatCurve=DataInterface() {propertyValues {Curve {keys [<at 0.5 1>,<at 1 0>]}}}, Scale Curve=0.2)
    )
    Particles.GlowSize=&Multiply_Vector2D(A=$Particles.SpriteSize, B=<8 8>)
   ) {isSetParameters=true}
  ]
  rendererClasses [&NiagaraSpriteRendererProperties]
  renderers [
   {rendererClass=&NiagaraSpriteRendererProperties}
   {rendererClass=&NiagaraSpriteRendererProperties, rendererIndex=1}
  ]
 }
 Tris {
  emitterUpdateScript [
   &EmitterState(Life Cycle Mode=::NewEnumerator1, Loop Behavior=::NewEnumerator1, Loop Duration=1, UseLoopDelay=false)
   &SpawnBurst_Instantaneous(
    Spawn Count=&RandomRangeInt(Minimum=8, Maximum=10)
    Spawn Group=2
    Use Spawn Probability=false
    UseLoopCountLimit=false
   ) {module=SpawnBurst_Instantaneous002}
  ]
  particleSpawnScript [
   /SetVariables(Particles.VisibilityTag=&MergeHelper_IntSwitch(0=0, 1=1, 2=2, 3=?, 4=?, 5=?, Index=$Emitter.SpawnGroup, Num=::NewEnumerator1)) {isSetParameters=true}
   /SetVariables(
    Particles.CustomRandom:NiagaraFloat=&RandomRangeFloat(Minimum=0.2, Maximum=0.8)
    Particles.CustomRandomTriangle:NiagaraFloat=&RandomRangeFloat(Minimum=0, Maximum=0.3)
   ) {isSetParameters=true}
   &InitializeParticle(
    Color=&MultiplyLinearColor(
     Linear Color=&MergeHelper_ColorSwitch(
      0=&RandomRangeLinearColor(Minimum=<rgba 2 1.725 0.7396 1>, Maximum=<rgba 2 0.9554 0 1>)
      1=<rgba 0 0 0 0>
      2=&Multiply_LinearColorByFloat(Linear Color=<rgba 1 1 1 1>, Float=&Lerp_Float(A=1, B=2, Alpha=$Particles.CustomRandomTriangle))
      3=?
      Index=$Particles.VisibilityTag
      Num=::NewEnumerator1
     )
     Scale Factor=&MakeVector4_FromLinearColor(LinearColor=$System.TOD)
    )
    Lifetime=&MergeHelper_FloatSwitch(
     0=&RandomRangeFloat(Minimum=0.4, Maximum=0.6)
     1=&RandomRangeFloat(Minimum=0.4, Maximum=0.65)
     2=&RandomRangeFloat(Minimum=0.15, Maximum=0.275)
     3=&RandomRangeFloat(Minimum=0.1, Maximum=0.1)
     Index=$Emitter.SpawnGroup
     Num=::NewEnumerator1
    )
    Mesh Scale=&Multiply_VectorByFloat(
     Vector=&MakeVector(
      X=&RandomRangeFloat(Minimum=1, Maximum=1)
      Y=&RandomRangeFloat(Minimum=0.2, Maximum=2)
      Z=&RandomRangeFloat(Minimum=1, Maximum=1)
     )
     Float=&Multiply_Float(
      A=&RandomRangeFloat(Minimum=0.2, Maximum=0.45)
      B=&FN_FrontEndFloatModifier(Gameplay=1, Front End=$User.FrontEndScale)
     )
    )
    Mesh Scale Mode=::NewEnumerator4
    UsePositionOffset=false
    Write Material Random=true
   )
   /SetVariables(Particles.InitialSize:Vector2f=&Vector2DFromFloat(Value=&Multiply_Float(
      A=&RandomRangeFloat(Minimum=20, Maximum=30)
      B=&FN_FrontEndFloatModifier(Gameplay=1, Front End=$User.FrontEndScale)
     ))) {isSetParameters=true}
   &InitialMeshOrientation(Use Rotation Vector=false)
   /SetVariables(Particles.DynamicTriangles:Vector4f=&MakeVector4(X=&Lerp_Float(A=1, B=2, Alpha=$Particles.CustomRandomTriangle), Y=3, Z=1, W=1)) {isSetParameters=true}
   &ShapeLocation(
    MaskSpawnGroup=false
    Offset Mode=::NewEnumerator0
    Override Seed=false
    Sphere Radius=&Multiply_Float(A=10, B=&FN_FrontEndFloatModifier(Gameplay=1, Front End=$User.FrontEndScale))
    UseOwnerScale=true
   )
   &AddVelocity(
    Velocity=$Particles.ShapeLocation.ShapeNormal
    Velocity Speed Scale=&Multiply_Float(
     A=&RandomRangeFloat(Minimum=180, Maximum=250)
     B=&FN_FrontEndFloatModifier(Gameplay=1, Front End=$User.FrontEndScale)
    )
   ) {module=AddVelocity001}
  ]
  particleUpdateScript [
   &Drag(
    Drag=&FloatFromCurve(
     FloatCurve=DataInterface() {propertyValues {Curve {keys [
         {arriveTangent=2.0771, interpMode=RCIM_Cubic, leaveTangent=2.0771, tangentMode=RCTM_User, time=0, value=0}
         {interpMode=RCIM_Cubic, tangentMode=RCTM_User, time=1, value=1}
        ]}}}
     Scale Curve=10
    )
    Rotational Drag:NiagaraFloat=&FloatFromCurve(
     FloatCurve=DataInterface() {propertyValues {Curve {keys [
         {arriveTangent=2.0771, interpMode=RCIM_Cubic, leaveTangent=2.0771, tangentMode=RCTM_User, time=0, value=0}
         {interpMode=RCIM_Cubic, tangentMode=RCTM_User, time=1, value=1}
        ]}}}
     Scale Curve=3
    )
    Use Linear Drag=true
    Use Rotational Drag=true
   ) {module=Drag001}
   &SolveForcesAndVelocity(Clamp Velocity=false, Limit Acceleration=false)
   &ParticleState()
   &UpdateMeshOrientation(
    Rotation Rate=&RandomRangeFloat(Minimum=0.5, Maximum=1)
    Rotation Vector=&RandomRangeVector(Minimum=<0 0 0>, Maximum=<0 1 1>)
   )
   &ScaleColor(
    Scale Alpha=&MergeHelper_FloatSwitch(
     0=&FloatFromCurve(FloatCurve=DataInterface() {propertyValues {Curve {keys [<at 0 0>,<at 0.1 1>,<at 0.5 1>,<at 1 0>]}}})
     1=0
     2=&Multiply_Float(
      A=0.05
      B=&FloatFromCurve(
       FloatCurve=DataInterface() {propertyValues {Curve {keys [{interpMode=RCIM_Cubic, tangentMode=RCTM_User, time=0, value=1}, <at 1 0.5>]}}}
       Scale Curve=0.5
      )
     )
     3=?
     Index=$Particles.VisibilityTag
     Num=::NewEnumerator1
    )
    Scale RGB=&MergeHelper_VectorSwitch(
     0=<1 1 1>
     1=<1 1 1>
     2=&MakeVectorFromLinearColorRGB(Color=&ColorFromCurve(ColorCurve=DataInterface() {propertyValues {
         AlphaCurve {keys [<at 0 1 RCIM_Cubic>]}
         BlueCurve {keys [<at 0 10 RCIM_Cubic>,<at 1 1>]}
         GreenCurve {keys [<at 0 6 RCIM_Cubic>,<at 1 1>]}
         RedCurve {keys [<at 0 2 RCIM_Cubic>,<at 1 1>]}
        }}))
     3=?
     Index=$Particles.VisibilityTag
     Num=::NewEnumerator1
    )
    ScaleA=true
    ScaleRGB=true
   )
  ]
  rendererClasses [&NiagaraMeshRendererProperties]
  renderers [{rendererClass=&NiagaraMeshRendererProperties, rendererIndex=0}]
 }
}
summary {
 systemName=NS_Emote_WealthLamb_Fate_Impact
 emitters [
  {emitterName=Sparks_CPU, rendererClasses [&NiagaraSpriteRendererProperties]}
  {emitterName=ChromaRadialFlare, rendererClasses [&NiagaraSpriteRendererProperties]}
  {emitterName=Tris, rendererClasses [&NiagaraMeshRendererProperties]}
  {
   emitterName=FlashElements001
   rendererClasses [&NiagaraSpriteRendererProperties, &NiagaraLightRendererProperties]
  }
  {emitterName=CrystalFragments, rendererClasses [&NiagaraMeshRendererProperties]}
 ]
 userVariables [
  User.bIsFrontEnd {type=NiagaraBool}
  User.bIsFrontEndPreview {type=NiagaraBool}
  User.FrontEndScale {type=NiagaraFloat, defaultValue=0.6, description=""}
  User.SwingSectionIndex {type=NiagaraInt32}
 ]
}
