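# Test Niagara system in the WealthLamb_FateStork emote folder that exercises a boolean-driven timeline on sprite particles.
# Particles spawn on a sphere around a sampled skeletal mesh position and are attracted back toward it.
# Scratch-pad module BisBreakingScratchPad takes that position, a distance and timing, outputting a BooleanTimeline that shifts color from blue to red.
# The age-based bool and NMS_TimelineFromBool variant are disabled.
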NiagaraSystem /BRCosmetics/Animation/Game/MainPlayer/Emotes/WealthLamb_FateStork/VFX/NS_TEST_NMSTimeline
dependencies {
 usedDataInterfaces [&NiagaraDataInterfaceCurve, &NiagaraDataInterfaceSkeletalMesh]
 usedDynamicInputs []
 usedModules [
  &SystemState
  &self:SystemSpawnScript.SystemScriptSource.SystemScriptGraph.NiagaraNodeAssignment_0.SetVariables_#1
 ]
}
emitters {Empty {
  emitterUpdateScript [
   &EmitterState()
   &SpawnRate(SpawnRate=20.531, Use Spawn Probability=false)
  ]
  particleSpawnScript [
   &InitializeParticle(
    Color=<rgba 1 1 1 10>
    Lifetime Max=6
    Lifetime Min=4
    Lifetime Mode=::NewEnumerator1
    Sprite Size Mode=::NewEnumerator3
    UsePositionOffset=false
    Write Material Random=true
   )
   &SkeletalMeshLocation(
    Bone / Socket Index=3
    MaskSpawnGroup=false
    Mesh Sampling Type=::NewEnumerator5
    Orientation Sampling=::NewEnumerator1
    Position Sampling=::NewEnumerator2
    Sampled Position Offset Coordinate Space:ENiagaraSkelMeshTransforms=::NewEnumerator2
    Skeletal Mesh=$System.Skeletal Mesh
    Socket Sampling Mode=::NewEnumerator5
    Use Sampled Mesh Attribute as Ribbon ID=false
   ) {module=SkeletalMeshLocation001}
   &ShapeLocation(
    MaskSpawnGroup=false
    Override Seed=false
    Shape Origin=$Particles.SkeletalMeshLocation001.SampledPosition
    Sphere Radius=150
    Sphere Surface Distribution=1
    UseOwnerScale=true
   )
  ]
  particleUpdateScript [
   &ParticleState()
   &self:BisBreakingScratchPad(
    BreakingPosition:NiagaraPosition=$Particles.SkeletalMeshLocation001.SampledPosition
    DistanceToBreakPos=50
    Duration=2
    SetTrueAfterTime=0.75
   ) {module=BisBreakingScratchPad}
   /SetVariables(Particles.bIsTrue:NiagaraBool=&SetBoolByFloatComparison(A=$Particles.Age, B=2)) {enabled=false, isSetParameters=true}
   &/NMS_TimelineFromBool(Duration=2, Module_Condition:NiagaraBool=$Particles.bIsTrue) {enabled=false}
   &PointAttractionForce(
    Attraction Radius=150
    Attraction Strength=5
    Attractor Position:NiagaraPosition=$Particles.SkeletalMeshLocation001.SampledPosition
    Kill Within Radius=false
    Module.Use Falloff=true
    UseAttractorOffset=false
   )
   &SolveForcesAndVelocity(Clamp Velocity=false, Limit Acceleration=false)
   &ScaleColor(
    Scale RGB=&MakeVector(
     X=&FloatFromCurve(FloatCurve=DataInterface() {propertyValues {Curve {keys [<at 0 0>,<at 1 1>]}}}, CurveIndex=$Output.BisBreakingScratchPad.BooleanTimeline)
     Y=0
     Z=&FloatFromCurve(FloatCurve=DataInterface() {propertyValues {Curve {keys [<at 0 1>,<at 1 0>]}}}, CurveIndex=$Output.BisBreakingScratchPad.BooleanTimeline)
    )
    ScaleA=true
    ScaleRGB=true
   )
  ]
  rendererClasses [&NiagaraSpriteRendererProperties]
  renderers [{rendererClass=&NiagaraSpriteRendererProperties}]
 }}
summary {
 systemName=NS_TEST_NMSTimeline
 emitters [{emitterName=Empty, rendererClasses [&NiagaraSpriteRendererProperties]}]
 userVariables []
}
