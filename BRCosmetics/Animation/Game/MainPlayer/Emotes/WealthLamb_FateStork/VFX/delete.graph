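# Additive unlit particle material for the WealthLamb_FateStork emote VFX, drawing a flare tinted by a scrolling color texture, rainbow by default.
# Polar UVs scroll with time plus a per-particle offset, using constant sample derivatives to hide the seam; a radial mask desaturates color toward the edges.
# That color, modulated by particle color and FlareMask plus a dim base, feeds an emissive adjustment function.
# Particle alpha times flare green gets depth fade, skipped on low/mobile.
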Material /BRCosmetics/Animation/Game/MainPlayer/Emotes/WealthLamb_FateStork/VFX/delete
graph {
 properties {
  bUseTranslucencyVertexFog=false
  bUsedWithNiagaraSprites=true
  bUsedWithParticleSprites=true
  blendMode=BLEND_Additive
  pixelDepthOffsetMode=PDOM_Legacy
  shadingModel=MSM_Unlit
  usedShadingModels=Unlit
 }
 nodes [
  n1 = TextureCoordinate()
  DDX(Value=@n1)
  DDY(Value=@n1)
  n2 = MaterialFunctionCall!(UV Coord=TextureCoordinate(), LenghtScale, RadialScale) {materialFunction=/Game/Global/MaterialFunctions/TechArt/MF_PolarCoordinates}
  n3 = VectorParameter() {
   desc="Color packing UV Tile and scroll speed\r\nG channel should be even numbers otherwise you will see a seam "
   group=Color
   parameterName=Color UV
   channelNames {a=Scroll Speed V, b=Scroll Speed U, g=Tile V, r=Tile U}
   defaultValue=<rgba 0.2 1 -0.05 0>
  }
  n4 = MaterialFunctionCall(X=@n3.Tile U, Y=@n3.Tile U) {materialFunction=&MakeFloat2}
  n5 = MaterialFunctionCall(X=@n3.Scroll Speed U, Y=@n3.Scroll Speed U) {materialFunction=&MakeFloat2}
  n6 = DynamicParameter() {parameterIndex=1, defaultValue=<rgba 0 0 1 0>, paramNames [UVTilingRandom, TimeOffsetRandom, WPOStrength, TrailingModeON]}
  n7 = Add(A=@n6.TimeOffsetRandom, B=Time())
  n8 = Constant() {desc=Fixes Polar Coor Seam}
  n9 = TextureSampleParameter2D!(UVs=Add(A=Multiply(A=@n2.Result, B=@n4.Result), B=Multiply(A=@n5.Result, B=@n7)), "DDX(UVs)"=@n8, "DDY(UVs)"=@n8, Apply View MipBias) {group=Color, mipValueMode=TMVM_Derivative, parameterName=Color Tex, texture=&T_Tile_RainbowColor}
  n10 = TextureCoordinate()
  n11 = ScalarParameter() {defaultValue=0.5, group=Color}
  n12 = Desaturation(
   None=@n9.RGB
   Fraction=LinearInterpolate(
    A=ScalarParameter() {defaultValue=1, group=Color, parameterName=Color Tex Desaturation}
    Alpha=SphereMask(A=@n10, B=@n11, Radius=@n11) {hardnessPercent=0}
   ) {constB=0}
  )
  n13 = ParticleColor()
  n14 = TextureSampleParameter2D!(UVs, Apply View MipBias) {
   parameterName=FlareMask
   samplerType=SAMPLERTYPE_LinearColor
   texture=/Game/Effects/Fort_Effects/GameplayPulginsAssets/DiamondSkin/Textures/T_DiamondSkin_Flare
  }
  n15 = Multiply(A=@n13.RGB, B=@n14.RGB)
  MaterialFunctionCall(Emissive=Add(A=Multiply(A=@n12) {constB=0.25}, B=Multiply(A=@n12, B=@n15))) {materialFunction=&MF_FX_EmissiveAdjustments}
  n16 = Multiply(A=@n13.A, B=@n14.G)
  n17 = ScalarParameter() {parameterName=DepthFade}
  MaterialFunctionCall!(Default=Multiply(A=@n16, B=DepthFade(FadeDistance=@n17)), Low/Mobile=@n16) {materialFunction=/Game/Effects/Fort_Effects/Materials/_Functions/MF_LowMobile_PathingSwitch}
 ]
}
